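# M_PrismParticle_Glitch is an unlit, masked Niagara particle material for a holographic glitch effect on the PrismParticle character body.
# Atlas hologram wiggle drives optional vertex offset and scanline block opacity.
# Particle-colored scanlines combine with two sine-swept screen-space lines, then blend lightly with noise for emissive.
# The opacity mask multiplies those lines by noise-based erosion, with dynamic parameters controlling wiggle, line timing, erosion and noise scale.
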Material /BRCosmetics/Characters/Player/Male/Medium/Bodies/M_MED_PrismParticle/FX/M_PrismParticle_Glitch
graph {
 properties {
  bUsedWithNiagaraMeshParticles=true
  bUsedWithNiagaraRibbons=true
  bUsedWithNiagaraSprites=true
  blendMode=BLEND_Masked
  pixelDepthOffsetMode=PDOM_Legacy
  shadingModel=MSM_Unlit
  usedShadingModels=Unlit
 }
 nodes [
  n1 = MaterialFunctionCall!(
   WPO Intensity A
   WPO Intensity B
   Time=Multiply(A=Time()) {constB=0.2}
   Bias=DynamicParameter() {defaultValue=<rgba 0 5 1 1>, paramNames [TriggerWiggle, ScanLineRandom, Erosion, ErosionScale]}.TriggerWiggle
  ) {materialFunction=/Game/Effects/Fort_Effects/Effects/Atlas/Materials/MF_Atlas_HologramWiggle_01}
  n2 = Constant()
  n3 = MaterialFunctionCall!(
   Number of Scanlines=ScalarParameter() {defaultValue=250, parameterName=Scanline Amount}
   Scanline Speed=ScalarParameter() {defaultValue=-0.05, parameterName=Scanline Speed}
   "Scanlines Thickness (Bias)"=ScalarParameter() {defaultValue=-0.15, parameterName=ScanlineSize}
   Scanlines Dark Value=ScalarParameter() {defaultValue=0.25, parameterName=Scanlines Dark}
   Scanlines Light Value=ScalarParameter() {defaultValue=1, parameterName=Scanlines Light}
   Block Opacity=Multiply(A=StaticSwitchParameter(True=@n1.Intense 0-1, False=@n2) {defaultValue=true, parameterName=Use Blocks?}) {constB=0.333}
   UVs
   Time
   Block Repeat
  ) {materialFunction=/Game/Effects/Fort_Effects/Effects/Atlas/Materials/MF_Atlas_Scanlines_01}
  n4 = Multiply(A=ParticleColor().RGB, B=@n3.Result)
  n5 = MaterialFunctionCall(World Position=WorldPosition().XYZ) {materialFunction=&TransformToClipSpace}
  n6 = Add(
   A=Time()
   B=DynamicParameter() {defaultValue=<rgba 0 5 1 1>, paramNames [TriggerWiggle, ScanLineRandom, Erosion, ErosionScale]}.ScanLineRandom
  )
  n7 = Subtract(A=@n5.Screen aligned 0-1 UV, B=ConstantBiasScale(None=Sine(None=@n6) {period=5.41}) {bias=1.3, scale=0.35})
  n8 = ScalarParameter() {defaultValue=96, parameterName=Line Multiplier A}
  n9 = StaticComponentMaskParameter(None=Clamp(None=Multiply(A=@n7, B=@n8))) {defaultG=true}
  n10 = Multiply(A=@n9, B=OneMinus(None=@n9))
  n11 = Subtract(A=@n5.Screen aligned 0-1 UV, B=ConstantBiasScale(None=Sine(None=@n6) {period=7.89}) {bias=1.25, scale=0.3})
  n12 = ScalarParameter() {defaultValue=84, parameterName=Line Multiplier B}
  n13 = StaticComponentMaskParameter(None=Clamp(None=Multiply(A=@n11, B=@n12))) {defaultG=true}
  n14 = Multiply(A=@n13, B=OneMinus(None=@n13))
  n15 = Add(
   A=MaterialFunctionCall!(In=Multiply(A=@n10) {constB=2.1}) {materialFunction=/Engine/Functions/Engine_MaterialFunctions02/Utility/Round}.Result
   B=MaterialFunctionCall(In=Multiply(A=@n14) {constB=2.2}) {materialFunction=&Round}.Result
  )
  n16 = DynamicParameter() {defaultValue=<rgba 0 5 1 1>, paramNames [TriggerWiggle, ScanLineRandom, Erosion, ErosionScale]}
  n17 = TextureSample(UVs=Multiply(A=TextureCoordinate(), B=@n16.ErosionScale)) {texture=/Game/Athena/UI/Challenges/Art/T_UI_DataChip_Noise}
  MaterialFunctionCall(Emissive=LinearInterpolate(A=Add(A=@n4, B=@n15), B=@n17.RGB) {constAlpha=0.2}) {materialFunction=&MF_FX_EmissiveAdjustments}
  Multiply(A=Add(A=@n15, B=@n3.Result), B=LinearInterpolate(A=Subtract(A=@n17.G) {constB=0.5}, Alpha=@n16.Erosion) {constAlpha=0})
  StaticSwitchParameter(True=@n1.Result, False=Constant4Vector().RGBA) {defaultValue=true, parameterName=Use WPO}
 ]
}
