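# Niagara sprite effect that adds flickering glitch bursts for the hair of the PrismParticle player body cosmetic.
# One looping emitter fires random-sized bursts, gated by GlitchSpawnProbability, on a small offset sphere with short lifetimes, random drift and drag.
# Color blends pink to cyan by camera angle to the sphere normal, dimmed by time of day.
# Curves jitter sprite size and feed dynamic material parameters, while bisHeadHidden sets each particle's VisibilityTag.
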NiagaraSystem /BRCosmetics/Characters/Player/Male/Medium/Bodies/M_MED_PrismParticle/FX/NS_PrismParticle_GlitchHair
dependencies {
 usedDataInterfaces [&NiagaraDataInterfaceCamera, &NiagaraDataInterfaceCurve]
 usedDynamicInputs []
 usedModules [&SystemState]
}
emitters {Glitch_Hair {
  emitterUpdateScript [
   &EmitterState(
    Life Cycle Mode=::NewEnumerator1
    Loop Delay=0.05
    Loop Duration=&RandomRangeFloat(Minimum=0, Maximum=1.4, Evaluation Type=::NewEnumerator1)
    Recalculate Duration Each Loop=true
    UseLoopDelay=true
   )
   &SpawnBurst_Instantaneous(
    Spawn Count=&RandomRangeInt(Minimum=0, Maximum=12)
    Spawn Probability=$User.GlitchSpawnProbability
    Use Spawn Probability=true
    UseLoopCountLimit=false
   )
  ]
  particleSpawnScript [
   &InitializeParticle(
    Lifetime Max=0.5
    Lifetime Min=0.06
    Lifetime Mode=::NewEnumerator1
    Sprite Size Max=<10 7>
    Sprite Size Min=<1 0.5>
    Sprite Size Mode=::NewEnumerator2
    UsePositionOffset=true
    Write Material Random=true
   ) {module=InitializeParticle001}
   &ShapeLocation(
    MaskSpawnGroup=false
    Offset=<5 -2.5 0>
    Offset Coordinate Space=::Local
    Offset Mode=::NewEnumerator0
    Override Seed=false
    Sphere Radius=8
    Sphere Surface Distribution=0.9
    UseOwnerScale=true
   ) {module=ShapeLocation001}
   &AddVelocity(Velocity=&RandomRangeVector(Minimum=<-20 -20 -20>, Maximum=<20 20 20>)) {module=AddVelocity001}
   &AddVelocity(
    Module.Use Falloff Distance=false
    Override Seed=false
    Velocity Mode=::NewEnumerator1
    Velocity Speed=&RandomRangeFloat(Minimum=2, Maximum=16)
   )
   &CameraOffset(Camera Offset Amount=3) {module=CameraOffset001}
  ]
  particleUpdateScript [
   &ParticleState()
   &Drag(Drag=8, Use Linear Drag=true, Use Rotational Drag=false)
   &SolveForcesAndVelocity(Clamp Velocity=false, Limit Acceleration=false)
   /SetVariables(Particles.Color=&MultiplyLinearColor(
     Linear Color=&LerpLinearColors(
      StartColor=<rgba 0.9 0.09 0.495 1>
      EndColor=<rgba 0.1 0.82 1 1>
      LerpFactor=&Power(
       a=&Abs_Float(Float:NiagaraFloat=&DotProduct(
         Vector A=&ReturnCameraVector(Camera Vectors=::NewEnumerator4, Coordinate Space=::Simulation, Origin Position=$Particles.Position)
         Vector B=$Particles.ShapeLocation001.ShapeNormal
        ))
       b=1
      )
     )
     Scale Factor=&MakeVector4_FromFloat(Float=&FN_Select_Float_from_Time_Of_Day(Morning=0.9, Day=1, Evening=0.6, Night=0.9))
    )) {isSetParameters=true}
   &ScaleSpriteSize(
    Scale Sprite Size Mode=::NewEnumerator0
    Uniform Scale Factor=&FloatFromCurve(
     FloatCurve=DataInterface() {propertyValues {Curve {keys [
         {arriveTangent=5.4245, time=0, value=5}
         {arriveTangent=-3.9614, time=0.0963, value=8.5355}
         {arriveTangent=-22.2662, leaveTangent=-22.2662, time=0.2, value=4}
         <at 0.3 4>
         <at 0.4 11.5>
         {arriveTangent=0.1532, leaveTangent=0.1532, time=0.5, value=2.875}
         <at 0.6 3.5>
         <at 0.7 8.5>
         <at 0.8 2.5>
         <at 0.9 3.5>
         {arriveTangent=-0.0405, leaveTangent=-0.0405, time=1, value=0}
        ]}}}
     Scale Curve=0.1
    )
   )
   &DynamicMaterialParameters(
    Index 0 Param 1=&FloatFromCurve(FloatCurve=DataInterface() {propertyValues {Curve {keys [
         {arriveTangent=4.1785, interpMode=RCIM_Cubic, leaveTangent=4.1785, tangentMode=RCTM_User, time=0, value=0}
         {arriveTangent=0.9995, interpMode=RCIM_Cubic, leaveTangent=0.9995, tangentMode=RCTM_User, time=0.2223, value=0.6825}
         {arriveTangent=-0.0405, interpMode=RCIM_Cubic, leaveTangent=-0.0405, tangentMode=RCTM_User, time=1, value=1}
        ]}}})
    Index 0 Param 2=&RandomRangeFloat(Minimum=-1, Maximum=0.1)
    Index 0 Param 3=&FloatFromCurve(
     FloatCurve=DataInterface() {propertyValues {Curve {keys [<at 0 1 RCIM_Cubic>, {arriveTangent=-1.8781, interpMode=RCIM_Cubic, leaveTangent=-1.8781, tangentMode=RCTM_User, time=1, value=0}]}}}
     Scale Curve=&RandomRangeFloat(Minimum=0.6, Maximum=0.9)
    )
    Index 0 Param 4=&FloatFromCurve(FloatCurve=DataInterface() {propertyValues {Curve {keys [<at 0 0.1>,<at 1 1>]}}})
    Module.Param0WriteEnabled=true
    Module.Param1WriteEnabled=true
    Module.Param2WriteEnabled=true
    Module.Param3WriteEnabled=true
   ) {module=DynamicMaterialParameters002}
   /SetVariables(Particles.VisibilityTag=&MakeCustomIntFromBool(Bool=$User.bisHeadHidden, True INT=1, False INT=0)) {isSetParameters=true}
  ]
  rendererClasses [&NiagaraSpriteRendererProperties]
  renderers [{rendererClass=&NiagaraSpriteRendererProperties}]
 }}
summary {
 systemName=NS_PrismParticle_GlitchHair
 emitters [{emitterName=Glitch_Hair, rendererClasses [&NiagaraSpriteRendererProperties]}]
 userVariables [
  User.bIsFrontEnd {type=NiagaraBool}
  User.bIsFrontEndPreview {type=NiagaraBool}
  User.bisHeadHidden {type=NiagaraBool}
  User.GlitchSpawnProbability {type=NiagaraFloat, defaultValue=1, description=""}
 ]
}
