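# Masked parent material for the PrismParticle character body, a holographic look with animated scanlines, glitch and flicker.
# Overrides base color, emissive, opacity mask and world position offset between the shared character PreFX and PostFX functions.
# Color combines curve-atlas remapping, time-of-day scanline tint, AO, fresnel and sphere-blended half-Lambert sun lighting.
# Scanlines use rotated pre-skinned position and stabilized screen-space UVs; digital noise drives glitch displacement, with temporal-dither opacity.
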Material /BRCosmetics/Characters/Player/Male/Medium/Bodies/M_MED_PrismParticle/Materials/M_PrismParticle_Parent
graph {
 properties {
  bUseMaterialAttributes=true
  bUsedWithMeshDeformer=true
  bUsedWithNanite=true
  bUsedWithSkeletalMesh=true
  blendMode=BLEND_Masked
  pixelDepthOffsetMode=PDOM_Legacy
  shadingModel=MSM_FromMaterialExpression
  usedShadingModels=Default Lit | Subsurface | Clear Coat | Subsurface Profile | Cloth
 }
 nodes [
  n1 = Multiply(A=NamedRerouteUsage(), B=NamedRerouteUsage())
  Add(A=NamedRerouteUsage(), B=@n1)
  n2 = Time()
  n3 = NamedRerouteUsage()
  n4 = Constant() {r=4}
  n5 = TextureSample(UVs=Add(A=Multiply(A=@n2) {constB=0.5}, B=Multiply(A=@n3, B=@n4))) {
   samplerSource=SSM_Wrap_WorldGroupSettings
   samplerType=SAMPLERTYPE_Masks
   texture=/BRCosmetics/Characters/Player/Male/Medium/Bodies/M_MED_PrismParticle/Textures/T_PrismParticle_Scanlines
  }
  n6 = NamedRerouteUsage()
  n7 = Time()
  n8 = TextureSample(UVs=Subtract(A=Multiply(A=@n6) {constB=10}, B=Multiply(A=@n7) {constB=0.1})) {
   samplerSource=SSM_Wrap_WorldGroupSettings
   samplerType=SAMPLERTYPE_Masks
   texture=&T_PrismParticle_Scanlines
  }
  n9 = Multiply(A=@n5.G, B=@n8.R)
  n10 = Multiply(A=Time()) {constB=0.1}
  n11 = TextureSample(
   UVs=MaterialFunctionCall(X=Frac(None=@n10), Y=Sine(None=@n10)) {materialFunction=&MakeFloat2}.Result
   Tex=NamedRerouteUsage()
  ) {
   samplerSource=SSM_Wrap_WorldGroupSettings
   samplerType=SAMPLERTYPE_LinearGrayscale
   texture=&T_SimplePerlin_Greyscale
  }
  n12 = LinearInterpolate(A=Multiply(A=NamedRerouteUsage(), B=OneMinus(None=@n9)), Alpha=Step(Y=@n11.R, X=Constant() {r=0.5}) {constX=0.5})
  ComponentMask(None=@n12) {g=true}
  n13 = NamedRerouteUsage()
  n14 = Constant() {r=5}
  n15 = Multiply(A=@n13, B=@n14)
  Floor(None=@n15)
  Frac(None=Multiply(A=Time()) {constB=0.5})
  n16 = VectorParameter() {group=PrismParticle, parameterName=PP_Scanline1_Color, defaultValue=<rgba 0 0.7941 1 1>}
  n17 = Desaturation(None=@n16.RGB, Fraction=Constant() {r=-0.2})
  n18 = MaterialFunctionCall(Color_Morning=@n16.RGB, Color_Day=@n16.RGB, Color_Evening=@n17, Color_Night=@n17) {materialFunction=&Color_Time_of_Day}
  n19 = CurveAtlasRowParameter(CurveTime=NamedRerouteUsage()) {
   atlas=/BRCosmetics/Characters/Player/Male/Medium/Bodies/M_MED_PrismParticle/Textures/CA_PrismParticle
   curve=/BRCosmetics/Characters/Player/Male/Medium/Bodies/M_MED_PrismParticle/Textures/C_PrismParticle
   group=PrismParticle
   parameterName=PP_Curve
  }
  n20 = Multiply(
   A=Desaturation(
    None=NamedRerouteDeclaration(None=@n19) {name=RemapedColor, nodeColor=<rgba 1 0.1172 0 1>}
    Fraction=ScalarParameter() {defaultValue=-0.5, group=PrismParticle, parameterName=PP_Desaturation}
   )
   B=ScalarParameter() {defaultValue=1.5, group=PrismParticle, parameterName=PP_Brightness}
  ) {constB=10}
  n21 = Multiply(A=Power(Base=NamedRerouteUsage()) {constExponent=1.8}) {constB=3}
  n22 = Multiply(A=OneMinus(None=NamedRerouteUsage()), B=ScalarParameter() {defaultValue=0.5, group=PrismParticle, parameterName=PP_AO2Scanlines}) {constB=0.5}
  n23 = NamedRerouteDeclaration(None=Saturate(None=Add(A=@n21, B=@n22))) {name=AOFresnel, nodeColor=<rgba 0.5469 0 1 1>}
  n24 = MaterialFunctionCall(Float3=NamedRerouteUsage()) {materialFunction=&BreakOutFloat3Components}
  n25 = Multiply(
   A=LinearInterpolate(A=NamedRerouteUsage(), B=LinearInterpolate(A=@n18.Result, B=@n20, Alpha=@n23), Alpha=NamedRerouteUsage()) {desc=ws scanlines}
   B=LinearInterpolate(B=@n24.G, Alpha=ScalarParameter() {defaultValue=0.15, group=PrismParticle, parameterName=PP_AO_Intensity}) {constA=1, constAlpha=0.3}
  )
  n26 = VectorParameter() {group=PrismParticle, parameterName=PP_ShadowTintColor, defaultValue=<rgba 0.7319 0.0582 0.8594 1>}
  n27 = NamedRerouteUsage()
  n28 = Multiply(
   A=Power(Base=OneMinus(None=@n27), Exp=ScalarParameter() {defaultValue=5, group=PrismParticle, parameterName=PP_ShadowTintPower})
   B=ScalarParameter() {defaultValue=0.2, group=PrismParticle, parameterName=PP_ShadowTintIntensity}
  )
  n29 = NamedRerouteUsage()
  n30 = ScalarParameter() {defaultValue=0.75, group=PrismParticle, parameterName=PP_LitTintIntensity}
  n31 = NamedRerouteUsage()
  n32 = LinearInterpolate(
   A=LinearInterpolate(
    A=LinearInterpolate(A=@n25, B=@n26.RGB, Alpha=@n28)
    B=VectorParameter() {group=PrismParticle, parameterName=PP_LitTintColor, defaultValue=<rgba 0.2023 0.9905 1 1>}.RGB
    Alpha=Multiply(A=@n29, B=@n30)
   )
   B=VectorParameter() {parameterName=c2, defaultValue=<rgba 0.5457 0.3005 0.5089 1>}.RGB
   Alpha=ComponentMask(None=@n31) {r=true}
  )
  LinearInterpolate(A=@n32, B=VectorParameter() {parameterName=c1, defaultValue=<rgba 0.5457 0.3005 0.5089 1>}.RGB, Alpha=ComponentMask(None=@n31) {g=true})
  n33 = MaterialFunctionCall() {materialFunction=&MF_CharacterParent_2023_PreFX}
  n34 = GetMaterialAttributes(None=@n33.Result) {attributeGetTypes [#1]}
  n35 = SetMaterialAttributes!(
   MaterialAttributes=@n34.MaterialAttributes
   Base Color=NamedRerouteUsage()
   Emissive Color=NamedRerouteUsage()
   Opacity Mask=NamedRerouteUsage()
   World Position Offset=NamedRerouteUsage()
  ) {attributeSetTypes [
    #2
    #3
    #4
    #5
   ]}
  MaterialFunctionCall(
   Material=@n35
   M=NamedRerouteDeclaration(None=@n33.M) {name=M_Map, nodeColor=<rgba 1 0 0.6328 1>}
   Pre Skinned Local Position=@n33.Pre skinned Local Position
   Distance Blend=@n33.Pre skinned Local Position
   Fresnel=@n33.Pre skinned Local Position
  ) {materialFunction=&MF_CharacterParent_2023_PostFX}
  MaterialFunctionCall(X=@n5.G, Y=Constant() {r=1}) {materialFunction=&MakeFloat2}
  n36 = Multiply(A=Time(), B=LinearInterpolate(B=ScalarParameter() {defaultValue=1, parameterName=PP_FlickerSpeed}, Alpha=NamedRerouteUsage())) {constB=1.5}
  n37 = TextureSample(
   UVs=MaterialFunctionCall(X=Frac(None=@n36), Y=Sine(None=@n36)) {materialFunction=&MakeFloat2}.Result
   Tex=NamedRerouteUsage()
  ) {
   samplerSource=SSM_Wrap_WorldGroupSettings
   samplerType=SAMPLERTYPE_LinearGrayscale
   texture=&T_SimplePerlin_Greyscale
  }
  MaterialFunctionCall(X=@n37.R, Y=@n37.R) {materialFunction=&MakeFloat2}
  NamedRerouteDeclaration(None=@n34.Normal) {name=Normal_IN, nodeColor=<rgba 0 1 0.3438 1>}
  n38 = Length(None=Subtract(A=ActorPositionWS(), B=CameraPositionWS())) {is=class MaterialExpressionLength}
  n39 = Saturate(None=SmoothStep(Value=Divide(A=@n38) {constB=1024}) {constMax=0.7, constMin=0.18})
  NamedRerouteDeclaration(None=@n39) {name=Debug, nodeColor=<rgba 0 0.5312 1 1>}
  n40 = Saturate(None=Distance(A=NamedRerouteUsage(), B=Constant()))
  n41 = Saturate(None=SmoothStep(
    Min=ScalarParameter() {defaultValue=0.5, group=PrismParticle, parameterName=PP_Scanline1_SolidThresholdMin}
    Max=ScalarParameter() {defaultValue=1, group=PrismParticle, parameterName=PP_Scanline1_SolidThresholdMax}
    Value=OneMinus(None=@n40)
   ) {constMax=0.5, constMin=0.3})
  n42 = LinearInterpolate(
   A=ScalarParameter() {defaultValue=0.8, group=PrismParticle, parameterName=PP_Scanline1_Thickness}
   B=ScalarParameter() {defaultValue=0.55, group=PrismParticle, parameterName=PP_Scanline1_Thickness2}
   Alpha=Power(Base=@n41) {constExponent=0.25}
  ) {constB=0}
  n43 = Saturate(None=Power(Base=@n41) {constExponent=10})
  n44 = RotateAboutAxis(
   NormalizedRotationAxis=Normalize(VectorInput=VectorParameter() {group=PrismParticle, parameterName=PP_Scanline1_RotationAxis, defaultValue=<rgba 0 1 0 1>}.RGB)
   RotationAngle=ScalarParameter() {defaultValue=-0.3, group=PrismParticle, parameterName=PP_Scanline1_Rotation}
   PivotPoint=Constant3Vector() {constant=<rgba 0 0 162 0>}
   Position=VertexInterpolator(VS=PreSkinnedPosition()).PS
  )
  n45 = ComponentMask(None=@n44) {b=true}
  n46 = Multiply(A=Time(), B=ScalarParameter() {defaultValue=-1, group=PrismParticle, parameterName=PP_Scanline1_Speed})
  n47 = ComponentMask(None=NamedRerouteUsage()) {g=true, r=true}
  n48 = ScalarParameter() {defaultValue=0.1, group=PrismParticle, parameterName=PP_Scanline1_NormalWarp}
  n49 = ComponentMask(None=@n44) {b=true, r=true}
  n50 = ScalarParameter() {defaultValue=15, group=PrismParticle, parameterName=PP_Scanline1_NoiseWarpScale}
  n51 = NamedRerouteDeclaration(None=TextureObject() {samplerType=SAMPLERTYPE_LinearGrayscale, texture=&T_SimplePerlin_Greyscale}) {name=PerlinTex, nodeColor=<rgba 0 1 0.8828 1>}
  n52 = TextureSample(UVs=Panner(Coordinate=Divide(A=@n49, B=@n50) {constB=16}) {speedX=0.1, speedY=-0.2}, Tex=@n51) {
   samplerSource=SSM_Wrap_WorldGroupSettings
   samplerType=SAMPLERTYPE_LinearGrayscale
   texture=&T_SimplePerlin_Greyscale
  }
  n53 = ComponentMask(None=@n44) {g=true, r=true}
  n54 = ScalarParameter() {defaultValue=20, group=PrismParticle, parameterName=PP_Scanline1_LineWarpScale}
  n55 = TextureSample(UVs=Panner(Coordinate=Divide(A=@n53, B=@n54)) {speedX=0.1, speedY=-0.2}, Tex=@n51) {
   samplerSource=SSM_Wrap_WorldGroupSettings
   samplerType=SAMPLERTYPE_LinearGrayscale
   texture=&T_SimplePerlin_Greyscale
  }
  n56 = Add(
   A=Add(A=Add(A=@n45, B=@n46), B=Multiply(A=@n47, B=@n48)) {desc=UV Warp}
   B=Multiply(A=Multiply(A=@n52.R, B=@n55.R), B=ScalarParameter() {defaultValue=1, group=PrismParticle, parameterName=PP_Scanline1_NoiseWarp})
  )
  n57 = ScalarParameter() {defaultValue=0.8, group=PrismParticle, parameterName=PP_Scanline1_Repitation}
  n58 = ConstantBiasScale(None=Sine(None=Divide(A=@n56, B=@n57) {constB=50}))
  n59 = MaterialFunctionCall(
   In=Multiply(A=LinearInterpolate(A=@n42, Alpha=@n43) {constB=0, desc=Tapering}, B=ComponentMask(None=@n58) {desc=Diagnal lines, r=true}) {constB=3}
   Contrast=ScalarParameter() {defaultValue=10, group=PrismParticle, parameterName=PP_Scanline1_Contrast}
  ) {materialFunction=&CheapContrast}
  NamedRerouteDeclaration(None=OneMinus(None=Saturate(None=@n59.Result))) {name=Scanlines1, nodeColor=<rgba 0 1 0.8125 1>}
  n60 = CollectionParameter() {
   collection=&FortniteMaterialParameters
   parameterName=SunlightDirectionalVector
  }
  n61 = PreSkinnedPosition()
  n62 = VectorParameter() {group=PrismParticle, parameterName=PP_NormalSphereLocation, defaultValue=<rgba 0 0 162 1>}
  n63 = VertexInterpolator(VS=Normalize(VectorInput=Subtract(A=@n61, B=@n62.RGB)))
  n64 = DotProduct(
   A=ComponentMask(None=@n60) {b=true, g=true, r=true}
   B=LinearInterpolate(
    A=PixelNormalWS()
    B=Transform(None=@n63.PS) {transformSourceType=TRANSFORMSOURCE_Local}
    Alpha=ScalarParameter() {defaultValue=0.5, group=PrismParticle, parameterName=PP_SphereNormalBlend}
   )
  )
  n65 = Power(
   Base=Add(A=Multiply(A=@n64) {constB=0.5}) {constB=0.5}
   Exp=ScalarParameter() {defaultValue=2, group=PrismParticle, parameterName=PP_HalfLambertPower}
  )
  n66 = ScalarParameter() {group=PrismParticle, parameterName=PP_LitBias}
  NamedRerouteDeclaration(None=Saturate(None=Add(A=@n65, B=@n66))) {name=HalfLambert, nodeColor=<rgba 0 1 0.5781 1>}
  n67 = ScalarParameter() {defaultValue=0.015, group=PrismParticle, parameterName=PP_Scanline2_Speed}
  n68 = LinearInterpolate(
   B=LinearInterpolate(B=ScalarParameter() {defaultValue=2, group=PrismParticle, parameterName=PP_UVScale_Far}, Alpha=@n39) {constA=1}
   Alpha=ScalarParameter() {group=PrismParticle, parameterName=PP_UseDistanceFadeUVScale}
  ) {constA=1}
  n69 = Time()
  n70 = MaterialFunctionCall!(
   Number of Scanlines=LinearInterpolate(A=ScalarParameter() {defaultValue=150, group=PrismParticle, parameterName=PP_Scanline2_Amount}, B=Constant(), Alpha=NamedRerouteUsage())
   Scanline Speed=LinearInterpolate(A=@n67, B=ScalarParameter() {defaultValue=0.5, group=PrismParticle, parameterName=PP_Scanline2_Speed_Glitched}, Alpha=NamedRerouteUsage())
   "Scanlines Thickness (Bias)"
   Scanlines Dark Value=ScalarParameter() {defaultValue=0.5, group=PrismParticle, parameterName=PP_Scanline2_Darkness}
   Scanlines Light Value
   Block Opacity=LinearInterpolate(
    A=ScalarParameter() {group=PrismParticle, parameterName=PP_Scanline2_BlockOpacity}
    B=ScalarParameter() {defaultValue=4, group=PrismParticle, parameterName=PP_Scanline2_BlockOpacity_Glitched}
    Alpha=NamedRerouteUsage()
   )
   UVs=NamedRerouteDeclaration(None=MaterialFunctionCall!(Position, UVRadialScale=@n68) {materialFunction=/Game/Weapons/WeaponSkins/MaterialFunction/ObjectStabilizedScreenSpaceUVs}.StabilizedUVs) {name=StabilizedUV, nodeColor=<rgba 0.4062 0 1 1>}
   Time=Multiply(A=Multiply(A=@n69, B=@n67)) {constB=20}
   Block Repeat=Constant() {r=3}
  ) {materialFunction=/Game/Effects/Fort_Effects/Effects/Atlas/Materials/MF_Atlas_Scanlines_01}
  n71 = Add(
   A=@n70.Result
   B=Multiply(A=@n70.Secondary Lines, B=ScalarParameter() {defaultValue=1, group=PrismParticle, parameterName=PP_AccentScanlightBrightness}) {constB=3}
  )
  n72 = Constant() {r=1}
  NamedRerouteDeclaration(None=LinearInterpolate(
    B=MaterialFunctionCall!(Default=@n71, Low/Mobile=@n72) {materialFunction=/Game/Effects/Fort_Effects/Materials/_Functions/MF_LowMobile_PathingSwitch}.Result
    Alpha=ScalarParameter() {defaultValue=0.3, group=PrismParticle, parameterName=PP_Scanline2_Intensity}
   ) {constA=1}) {name=Scanlines2, nodeColor=<rgba 0.4062 0 1 1>}
  n73 = Step(X=LinearInterpolate(Alpha=@n12)) {constX=0.5, constY=0.75}
  NamedRerouteDeclaration(None=Multiply(A=OneMinus(None=@n73), B=ScalarParameter() {defaultValue=0.6, group=PrismParticle, parameterName=PP_AmbientGlitchIntensity})) {name=AmbientGlitchOpacity, nodeColor=<rgba 1 0.7031 0 1>}
  n74 = PreSkinnedPosition()
  NamedRerouteDeclaration(None=VertexInterpolator(VS=@n74).PS) {name=PreskinnedPos, nodeColor=<rgba 1 0.9375 0 1>}
  NamedRerouteDeclaration(None=ChannelMaskParameter(None=@n33.Tech Art Mask) {group=PrismParticle, parameterName=PP_MaskChannel}.R) {name=HoloMask, nodeColor=<rgba 0 1 0.0156 1>}
  n75 = Divide(A=Ceil(None=@n15), B=@n14)
  n76 = Time()
  MaterialExpressionFrac_3 = Frac(None=Multiply(A=MaterialFunctionCall(X=@n76, Y=@n76) {materialFunction=&MakeFloat2}.Result) {constB=0.001})
  n77 = Custom!(xy=@MaterialExpressionFrac_3) {
   code="return float2(frac(sin(dot(xy, float2(12.9898,78.233)))*43758.5453123), frac(sin(dot(xy,float2(1234.321,324.0867)))*2345.21));"
   outputType=CMOT_Float1
   inputs [{
     inputName=xy
     input {expression=&self:MaterialExpressionFrac_3}
    }]
  }
  n78 = NamedRerouteDeclaration(None=MaterialFunctionCall(X=@n77, Y=@n77) {materialFunction=&MakeFloat2}.Result) {name=GlitchOffset, nodeColor=<rgba 0.875 0 1 1>}
  n79 = NamedRerouteUsage()
  n80 = Multiply(A=LinearInterpolate(A=Constant() {r=2}, Alpha=NamedRerouteUsage()) {constB=0}, B=@n79)
  n81 = TextureSample(UVs=Panner(Coordinate=Add(A=@n78, B=@n80)) {bFractionalPart=true, speedX=-0.876, speedY=0.654}) {
   samplerSource=SSM_Wrap_WorldGroupSettings
   samplerType=SAMPLERTYPE_LinearColor
   texture=/Game/Global/Textures/Default/Noise/T_Atlas_DigitalNoise_01
  }
  n82 = Power(Base=@n13) {constExponent=3}
  NamedRerouteDeclaration(None=MaterialFunctionCall(A=Constant() {r=1}, B=Multiply(A=@n75, B=@n81.R), C=Constant(), Alpha=Saturate(None=@n82)) {materialFunction=&Lerp_3Color}.Result) {name=GlitchOpacity, nodeColor=<rgba 1 0 0.7031 1>}
  NamedRerouteDeclaration(None=Step(Y=@n37.R) {constX=0.5}) {name=FlickerOpacity, nodeColor=<rgba 0 0.0156 1 1>}
  n83 = Constant() {r=0.5}
  n84 = TextureSample(UVs=Panner(Coordinate=Multiply(A=@n79, B=@n83)) {bFractionalPart=true, speedX=-1.5, speedY=0.987}) {
   samplerSource=SSM_Wrap_WorldGroupSettings
   samplerType=SAMPLERTYPE_LinearColor
   texture=&T_Atlas_DigitalNoise_01
  }
  n85 = Clamp(None=ConstantBiasScale(None=@n84.R) {bias=-0.5, scale=2}) {minDefault=-1}
  n86 = Clamp(None=ConstantBiasScale(None=@n84.G) {bias=-0.5, scale=2}) {minDefault=-1}
  n87 = Constant()
  n88 = ScalarParameter() {defaultValue=2, group=PrismParticle, parameterName=PP_GlitchWPOIntensity}
  n89 = NamedRerouteUsage()
  n90 = Transform(None=Multiply(
    A=MaterialFunctionCall(X=@n85, Y=@n86, Z=@n87) {materialFunction=&MakeFloat3}.Result
    B=LinearInterpolate(B=@n88, Alpha=@n89)
   )) {transformSourceType=TRANSFORMSOURCE_Local}
  n91 = Constant3Vector()
  n92 = Constant3Vector() {constant=<rgba 0 -4 140 0>}
  n93 = Saturate(None=Multiply(A=SphereMask(A=@n74, B=@n92) {attenuationRadius=15, hardnessPercent=0}) {constB=5})
  NamedRerouteDeclaration(None=StaticSwitchParameter(True=LinearInterpolate(A=@n90, B=@n91, Alpha=@n93) {constB=0}, False=@n90) {group=PrismParticle, parameterName=PP_UseWPOSphereMask}) {name=WPO_OUT, nodeColor=<rgba 0.2891 0 1 1>}
  NamedRerouteDeclaration(None=MaterialFunctionCall!(Alpha Threshold=@n1, Random) {materialFunction=/Game/Environments/Asteria/Sets/NeonCity/Techart/Screen/MaterialFunctions/DitherTemporalAABlue}.Result) {name=OpacityMask_OUT, nodeColor=<rgba 1 0.2344 0 1>}
  NamedRerouteDeclaration(None=MaterialFunctionCall(In=@n24.G, Contrast=ScalarParameter() {defaultValue=3, group=PrismParticle, parameterName=PP_AO_ScanlineContrast}) {materialFunction=&CheapContrast}.Result) {name=AO, nodeColor=<rgba 0 0.1094 1 1>}
  NamedRerouteDeclaration(None=ScalarParameter() {group=PrismParticle, parameterName=GlitchProgress}) {name=GlitchProgress, nodeColor=<rgba 0.8047 1 0 1>}
  n94 = Fresnel() {baseReflectFraction=0, exponent=1}
  n95 = ScalarParameter() {defaultValue=1, group=PrismParticle, parameterName=PP_FresnelPower}
  n96 = Multiply(A=Saturate(None=Power(Base=@n94, Exp=@n95)), B=ScalarParameter() {defaultValue=1, group=PrismParticle, parameterName=PP_FresnelIntensity})
  NamedRerouteDeclaration(None=OneMinus(None=NamedRerouteDeclaration(None=@n96) {name=Fresnel, nodeColor=<rgba 0 1 0.5547 1>})) {name=FresnelInverted, nodeColor=<rgba 1 0.0703 0 1>}
  n97 = Multiply(
   A=LinearInterpolate(A=@n32, B=VectorParameter() {group=PrismParticle, parameterName=PP_MaskedColor, defaultValue=<rgba 1 1 1 1>}.RGB, Alpha=NamedRerouteUsage())
   B=NamedRerouteUsage()
  )
  n98 = LinearInterpolate(A=@n97, B=Multiply(A=@n97, B=NamedRerouteUsage()), Alpha=ScalarParameter() {group=PrismParticle, parameterName=PP_InvertedFresnelIntensity})
  NamedRerouteDeclaration(None=Multiply(A=@n98, B=ScalarParameter() {defaultValue=0.1, group=PrismParticle, parameterName=PP_BaseColor_Intensity})) {name=BaseColor_OUT, nodeColor=<rgba 0 1 0.9297 1>}
  NamedRerouteDeclaration(None=Multiply(A=@n98, B=ScalarParameter() {defaultValue=0.8, group=PrismParticle, parameterName=PP_Emissive_Intensity})) {name=Emissive_OUT, nodeColor=<rgba 0 1 0.4609 1>}
  NamedRerouteUsage()
  NamedRerouteUsage()
  n99 = TransformPosition(None=ActorPositionWS()) {transformSourceType=TRANSFORMPOSSOURCE_World, transformType=TRANSFORMPOSSOURCE_View}
  SmoothStep(
   Min=ScalarParameter() {defaultValue=0.29}
   Max=ScalarParameter() {defaultValue=0.8, parameterName=Param1}
   Value=Divide(A=ObjectRadius(), B=ComponentMask(None=@n99) {b=true})
  ) {constMax=0.5}
  n100 = NamedRerouteUsage()
  n101 = Time()
  TextureSample(UVs=Subtract(A=Multiply(A=@n100) {constB=10}, B=Multiply(A=@n101) {constB=0})) {
   samplerSource=SSM_Wrap_WorldGroupSettings
   samplerType=SAMPLERTYPE_Masks
   texture=&T_PrismParticle_Scanlines
  }
 ]
}
