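# Physics asset for the M_MED_PrismParticle male player body cosmetic, set up for a dynamic long jacket over a lower-body and spine capsule proxy.
# Seven capsule chains (three back, two side, two front) run from MSTR roots with limited swing and twist, tightest twist at the back.
# Pelvis, spine, legs and feet link through free-angular constraints.
# One extra free-angular constraint ties the right front chain's fifth joint to the right side's, with no left-side counterpart.
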PhysicsAsset /BRCosmetics/Characters/Player/Male/Medium/Bodies/M_MED_PrismParticle/Meshes/M_MED_PrismParticle_Physics
bodies {
 calf_l [{shapeType=Capsule, length=32, radius=8, shapeName=calf_l_capsule, center=<-20 0 0>, rotation=<rot 90 0 0>}]
 calf_r [{shapeType=Capsule, length=32, radius=8, shapeName=calf_r_capsule, center=<20 0 0>, rotation=<rot 90 0 0>}]
 dyn_Jacket_bk_c_01_jnt [{shapeType=Capsule, length=6, radius=4, shapeName=dyn_Jacket_bk_c_01_jnt_capsule, center=<4 0 1>, rotation=<rot 0 0 90>}]
 dyn_Jacket_bk_c_02_jnt [{shapeType=Capsule, length=6, radius=4, shapeName=dyn_Jacket_bk_c_02_jnt_capsule, center=<4 0 1>, rotation=<rot 0 0 90>}]
 dyn_Jacket_bk_c_03_jnt [{shapeType=Capsule, length=10, radius=4, shapeName=dyn_Jacket_bk_c_03_jnt_capsule, center=<4 0 2>, rotation=<rot 0 0 90>}]
 dyn_Jacket_bk_c_04_jnt [{shapeType=Capsule, length=10, radius=5, shapeName=dyn_Jacket_bk_c_04_jnt_capsule, center=<4 0 2>, rotation=<rot 0 0 90>}]
 dyn_Jacket_bk_c_05_jnt [{shapeType=Capsule, length=10, radius=4, shapeName=dyn_Jacket_bk_c_05_jnt_capsule, center=<4 0 1>, rotation=<rot 0 0 90>}]
 dyn_Jacket_bk_c_06_jnt [{shapeType=Capsule, length=10, radius=6, shapeName=dyn_Jacket_bk_c_06_jnt_capsule, center=<5 0 1>, rotation=<rot 0 0 90>}]
 dyn_Jacket_bk_c_MSTR [{shapeType=Capsule, length=2, radius=2, shapeName=dyn_Jacket_bk_c_MSTR_capsule, center=<0 0 1>}]
 dyn_Jacket_bk_l_01_jnt [{shapeType=Capsule, length=6, radius=4, shapeName=dyn_Jacket_bk_l_01_jnt_capsule, center=<4 0 1>, rotation=<rot 0 0 90>}]
 dyn_Jacket_bk_l_02_jnt [{shapeType=Capsule, length=6, radius=4, shapeName=dyn_Jacket_bk_l_02_jnt_capsule, center=<4 0 1>, rotation=<rot 0 0 90>}]
 dyn_Jacket_bk_l_03_jnt [{shapeType=Capsule, length=10, radius=4, shapeName=dyn_Jacket_bk_l_03_jnt_capsule, center=<4 -3 2>, rotation=<rot 0 0 90>}]
 dyn_Jacket_bk_l_04_jnt [{shapeType=Capsule, length=10, radius=5, shapeName=dyn_Jacket_bk_l_04_jnt_capsule, center=<4 -3 2>, rotation=<rot 0 0 90>}]
 dyn_Jacket_bk_l_05_jnt [{shapeType=Capsule, length=10, radius=4, shapeName=dyn_Jacket_bk_l_05_jnt_capsule, center=<4 -3 1>, rotation=<rot 0 0 90>}]
 dyn_Jacket_bk_l_06_jnt [{shapeType=Capsule, length=10, radius=5, shapeName=dyn_Jacket_bk_l_06_jnt_capsule, center=<5 -3 1>, rotation=<rot 0 0 90>}]
 dyn_Jacket_bk_l_07_jnt [{shapeType=Capsule, length=10, radius=7, shapeName=dyn_Jacket_bk_l_07_jnt_capsule, center=<-4 -3 1>, rotation=<rot 0 0 90>}]
 dyn_Jacket_bk_l_MSTR [{shapeType=Capsule, length=2, radius=2, shapeName=dyn_Jacket_bk_l_MSTR_capsule, center=<0 0 1>}]
 dyn_Jacket_bk_r_01_jnt [{shapeType=Capsule, length=6, radius=4, shapeName=dyn_Jacket_bk_r_01_jnt_capsule, center=<4 0 1>, rotation=<rot 0 0 90>}]
 dyn_Jacket_bk_r_02_jnt [{shapeType=Capsule, length=6, radius=4, shapeName=dyn_Jacket_bk_r_02_jnt_capsule, center=<4 0 1>, rotation=<rot 0 0 90>}]
 dyn_Jacket_bk_r_03_jnt [{shapeType=Capsule, length=10, radius=4, shapeName=dyn_Jacket_bk_r_03_jnt_capsule, center=<4 3 2>, rotation=<rot 0 0 90>}]
 dyn_Jacket_bk_r_04_jnt [{shapeType=Capsule, length=10, radius=5, shapeName=dyn_Jacket_bk_r_04_jnt_capsule, center=<4 3 2>, rotation=<rot 0 0 90>}]
 dyn_Jacket_bk_r_05_jnt [{shapeType=Capsule, length=10, radius=4, shapeName=dyn_Jacket_bk_r_05_jnt_capsule, center=<4 3 1>, rotation=<rot 0 0 90>}]
 dyn_Jacket_bk_r_06_jnt [{shapeType=Capsule, length=10, radius=5, shapeName=dyn_Jacket_bk_r_06_jnt_capsule, center=<5 3 1>, rotation=<rot 0 0 90>}]
 dyn_Jacket_bk_r_07_jnt [{shapeType=Capsule, length=10, radius=7, shapeName=dyn_Jacket_bk_r_07_jnt_capsule, center=<-4 3 1>, rotation=<rot 0 0 90>}]
 dyn_Jacket_bk_r_MSTR [{shapeType=Capsule, length=2, radius=2, shapeName=dyn_Jacket_bk_r_MSTR_capsule, center=<0 0 1>}]
 dyn_Jacket_ft_l_01_jnt [{shapeType=Capsule, length=10, radius=3.5, shapeName=dyn_Jacket_ft_l_01_jnt_capsule, center=<4 -5 1>, rotation=<rot 0 0 60>}]
 dyn_Jacket_ft_l_02_jnt [{shapeType=Capsule, length=10, radius=3.5, shapeName=dyn_Jacket_ft_l_02_jnt_capsule, center=<6 -5 1>, rotation=<rot 0 0 60>}]
 dyn_Jacket_ft_l_03_jnt [{shapeType=Capsule, length=10, radius=4.5, shapeName=dyn_Jacket_ft_l_03_jnt_capsule, center=<8 -5 1>, rotation=<rot 0 0 60>}]
 dyn_Jacket_ft_l_04_jnt [{shapeType=Capsule, length=10, radius=5, shapeName=dyn_Jacket_ft_l_04_jnt_capsule, center=<6 -5 1>, rotation=<rot 0 0 60>}]
 dyn_Jacket_ft_l_05_jnt [{shapeType=Capsule, length=10, radius=5, shapeName=dyn_Jacket_ft_l_05_jnt_capsule, center=<4 -5 1>, rotation=<rot 0 0 60>}]
 dyn_Jacket_ft_l_MSTR [{shapeType=Capsule, length=2, radius=2, shapeName=dyn_Jacket_ft_l_MSTR_capsule, center=<0 0 1>}]
 dyn_Jacket_ft_r_01_jnt [{shapeType=Capsule, length=10, radius=3.5, shapeName=dyn_Jacket_ft_r_01_jnt_capsule, center=<4 5 1>, rotation=<rot 0 0 110>}]
 dyn_Jacket_ft_r_02_jnt [{shapeType=Capsule, length=10, radius=3.5, shapeName=dyn_Jacket_ft_r_02_jnt_capsule, center=<6 5 1>, rotation=<rot 0 0 110>}]
 dyn_Jacket_ft_r_03_jnt [{shapeType=Capsule, length=10, radius=4.5, shapeName=dyn_Jacket_ft_r_03_jnt_capsule, center=<8 5 1>, rotation=<rot 0 0 110>}]
 dyn_Jacket_ft_r_04_jnt [{shapeType=Capsule, length=10, radius=5, shapeName=dyn_Jacket_ft_r_04_jnt_capsule, center=<6 5 1>, rotation=<rot 0 0 110>}]
 dyn_Jacket_ft_r_05_jnt [{shapeType=Capsule, length=10, radius=5, shapeName=dyn_Jacket_ft_r_05_jnt_capsule, center=<4 5 1>, rotation=<rot 0 0 110>}]
 dyn_Jacket_ft_r_MSTR [{shapeType=Capsule, length=2, radius=2, shapeName=dyn_Jacket_ft_r_MSTR_capsule, center=<0 0 1>}]
 dyn_Jacket_l_01_jnt [{shapeType=Capsule, length=6, radius=4, shapeName=dyn_Jacket_l_01_jnt_capsule, center=<4 0 1>, rotation=<rot 0 0 90>}]
 dyn_Jacket_l_02_jnt [{shapeType=Capsule, length=6, radius=4, shapeName=dyn_Jacket_l_02_jnt_capsule, center=<4 0 1>, rotation=<rot 0 0 90>}]
 dyn_Jacket_l_03_jnt [{shapeType=Capsule, length=10, radius=5, shapeName=dyn_Jacket_l_03_jnt_capsule, center=<5 -2.5 2>, rotation=<rot 0 0.0005 79.9999>}]
 dyn_Jacket_l_04_jnt [{shapeType=Capsule, length=10, radius=5, shapeName=dyn_Jacket_l_04_jnt_capsule, center=<5 -2.5 2>, rotation=<rot 0 0.0005 79.9999>}]
 dyn_Jacket_l_05_jnt [{shapeType=Capsule, length=10, radius=5, shapeName=dyn_Jacket_l_05_jnt_capsule, center=<5 -2.5 2>, rotation=<rot 0 0.0005 79.9999>}]
 dyn_Jacket_l_06_jnt [{shapeType=Capsule, length=10, radius=5.5, shapeName=dyn_Jacket_l_06_jnt_capsule, center=<5 -2.5 2>, rotation=<rot 0 0.0005 79.9999>}]
 dyn_Jacket_l_07_jnt [{shapeType=Capsule, length=10, radius=6, shapeName=dyn_Jacket_l_07_jnt_capsule, center=<4 -2.5 2>, rotation=<rot 0 0.0005 80>}]
 dyn_Jacket_l_MSTR [{shapeType=Capsule, length=2, radius=2, shapeName=dyn_Jacket_l_MSTR_capsule, center=<0 0 1>}]
 dyn_Jacket_r_01_jnt [{shapeType=Capsule, length=6, radius=4, shapeName=dyn_Jacket_r_01_jnt_capsule, center=<4 0 1>, rotation=<rot 0 0 90>}]
 dyn_Jacket_r_02_jnt [{shapeType=Capsule, length=6, radius=4, shapeName=dyn_Jacket_r_02_jnt_capsule, center=<4 0 1>, rotation=<rot 0 0 90>}]
 dyn_Jacket_r_03_jnt [{shapeType=Capsule, length=10, radius=5, shapeName=dyn_Jacket_r_03_jnt_capsule, center=<5 2.5 2>, rotation=<rot 0 -0.0001 100>}]
 dyn_Jacket_r_04_jnt [{shapeType=Capsule, length=10, radius=5, shapeName=dyn_Jacket_r_04_jnt_capsule, center=<5 2.5 2>, rotation=<rot 0 -0.0001 100>}]
 dyn_Jacket_r_05_jnt [{shapeType=Capsule, length=10, radius=5, shapeName=dyn_Jacket_r_05_jnt_capsule, center=<5 2.5 2>, rotation=<rot 0 -0.0001 100>}]
 dyn_Jacket_r_06_jnt [{shapeType=Capsule, length=10, radius=5.5, shapeName=dyn_Jacket_r_06_jnt_capsule, center=<5 2.5 2>, rotation=<rot 0 -0.0001 100>}]
 dyn_Jacket_r_07_jnt [{shapeType=Capsule, length=10, radius=6, shapeName=dyn_Jacket_r_07_jnt_capsule, center=<4 2.5 2>, rotation=<rot 0 -0.0001 100>}]
 dyn_Jacket_r_MSTR [{shapeType=Capsule, length=2, radius=2, shapeName=dyn_Jacket_r_MSTR_capsule, center=<0 0 1>}]
 foot_l [{shapeType=Capsule, length=16, radius=8, shapeName=foot_l_capsule, center=<-4 6 0>, rotation=<rot 0 0 90>}]
 foot_r [{shapeType=Capsule, length=16, radius=8, shapeName=foot_r_capsule, center=<4 -6 0>, rotation=<rot 0 0 90>}]
 pelvis [{shapeType=Capsule, length=8, radius=11, shapeName=pelvis_capsule, center=<0 1.2089 0>}]
 spine_01 [{shapeType=Capsule, length=3, radius=10, shapeName=spine_01_capsule, center=<0 2.0374 0>}]
 spine_02 [{shapeType=Capsule, length=1, radius=9, shapeName=spine_02_capsule, center=<0 1.9354 0.7834>}]
 spine_03 [{shapeType=Capsule, length=2, radius=10, shapeName=spine_03_capsule, center=<0 2.5025 0>}]
 spine_04 [{shapeType=Capsule, length=4, radius=11.5, shapeName=spine_04_capsule, center=<0 3.3803 0>}]
 thigh_l [{shapeType=Capsule, length=32, radius=8.25, shapeName=thigh_l_capsule, center=<-20 0 0.5>, rotation=<rot 90 0 -5>}]
 thigh_r [{shapeType=Capsule, length=32, radius=8.25, shapeName=thigh_r_capsule, center=<20 0 -0.5>, rotation=<rot 90 0 -5>}]
}
constraints [
 {bone1Name=dyn_Jacket_bk_c_01_jnt, bone2Name=dyn_Jacket_bk_c_MSTR, swing1LimitDegrees=6, swing2LimitDegrees=16, twistLimitDegrees=2}
 {bone1Name=dyn_Jacket_bk_c_02_jnt, bone2Name=dyn_Jacket_bk_c_01_jnt, swing1LimitDegrees=6, swing2LimitDegrees=16, twistLimitDegrees=2}
 {bone1Name=dyn_Jacket_bk_c_03_jnt, bone2Name=dyn_Jacket_bk_c_02_jnt, swing1LimitDegrees=6, swing2LimitDegrees=16, twistLimitDegrees=2}
 {bone1Name=dyn_Jacket_bk_c_04_jnt, bone2Name=dyn_Jacket_bk_c_03_jnt, swing1LimitDegrees=6, swing2LimitDegrees=16, twistLimitDegrees=2}
 {bone1Name=dyn_Jacket_bk_c_05_jnt, bone2Name=dyn_Jacket_bk_c_04_jnt, swing1LimitDegrees=6, swing2LimitDegrees=16, twistLimitDegrees=2}
 {bone1Name=dyn_Jacket_bk_c_06_jnt, bone2Name=dyn_Jacket_bk_c_05_jnt, swing1LimitDegrees=6, swing2LimitDegrees=16, twistLimitDegrees=2}
 {bone1Name=dyn_Jacket_bk_l_01_jnt, bone2Name=dyn_Jacket_bk_l_MSTR, swing1LimitDegrees=6, swing2LimitDegrees=16, twistLimitDegrees=2}
 {bone1Name=dyn_Jacket_bk_l_02_jnt, bone2Name=dyn_Jacket_bk_l_01_jnt, swing1LimitDegrees=6, swing2LimitDegrees=16, twistLimitDegrees=2}
 {bone1Name=dyn_Jacket_bk_l_03_jnt, bone2Name=dyn_Jacket_bk_l_02_jnt, swing1LimitDegrees=6, swing2LimitDegrees=16, twistLimitDegrees=2}
 {bone1Name=dyn_Jacket_bk_l_04_jnt, bone2Name=dyn_Jacket_bk_l_03_jnt, swing1LimitDegrees=6, swing2LimitDegrees=16, twistLimitDegrees=2}
 {bone1Name=dyn_Jacket_bk_l_05_jnt, bone2Name=dyn_Jacket_bk_l_04_jnt, swing1LimitDegrees=6, swing2LimitDegrees=16, twistLimitDegrees=2}
 {bone1Name=dyn_Jacket_bk_l_06_jnt, bone2Name=dyn_Jacket_bk_l_05_jnt, swing1LimitDegrees=6, swing2LimitDegrees=16, twistLimitDegrees=2}
 {bone1Name=dyn_Jacket_bk_l_07_jnt, bone2Name=dyn_Jacket_bk_l_06_jnt, swing1LimitDegrees=6, swing2LimitDegrees=16, twistLimitDegrees=2}
 {bone1Name=dyn_Jacket_bk_r_01_jnt, bone2Name=dyn_Jacket_bk_r_MSTR, swing1LimitDegrees=6, swing2LimitDegrees=16, twistLimitDegrees=2}
 {bone1Name=dyn_Jacket_bk_r_02_jnt, bone2Name=dyn_Jacket_bk_r_01_jnt, swing1LimitDegrees=6, swing2LimitDegrees=16, twistLimitDegrees=2}
 {bone1Name=dyn_Jacket_bk_r_03_jnt, bone2Name=dyn_Jacket_bk_r_02_jnt, swing1LimitDegrees=6, swing2LimitDegrees=16, twistLimitDegrees=2}
 {bone1Name=dyn_Jacket_bk_r_04_jnt, bone2Name=dyn_Jacket_bk_r_03_jnt, swing1LimitDegrees=6, swing2LimitDegrees=16, twistLimitDegrees=2}
 {bone1Name=dyn_Jacket_bk_r_05_jnt, bone2Name=dyn_Jacket_bk_r_04_jnt, swing1LimitDegrees=6, swing2LimitDegrees=16, twistLimitDegrees=2}
 {bone1Name=dyn_Jacket_bk_r_06_jnt, bone2Name=dyn_Jacket_bk_r_05_jnt, swing1LimitDegrees=6, swing2LimitDegrees=16, twistLimitDegrees=2}
 {bone1Name=dyn_Jacket_bk_r_07_jnt, bone2Name=dyn_Jacket_bk_r_06_jnt, swing1LimitDegrees=6, swing2LimitDegrees=16, twistLimitDegrees=2}
 {bone1Name=dyn_Jacket_ft_l_01_jnt, bone2Name=dyn_Jacket_ft_l_MSTR, swing1LimitDegrees=2, swing2LimitDegrees=24, twistLimitDegrees=12}
 {bone1Name=dyn_Jacket_ft_l_02_jnt, bone2Name=dyn_Jacket_ft_l_01_jnt, swing1LimitDegrees=24, swing2LimitDegrees=24, twistLimitDegrees=12}
 {bone1Name=dyn_Jacket_ft_l_03_jnt, bone2Name=dyn_Jacket_ft_l_02_jnt, swing1LimitDegrees=32, swing2LimitDegrees=32, twistLimitDegrees=12}
 {bone1Name=dyn_Jacket_ft_l_04_jnt, bone2Name=dyn_Jacket_ft_l_03_jnt, swing1LimitDegrees=32, swing2LimitDegrees=32, twistLimitDegrees=12}
 {bone1Name=dyn_Jacket_ft_l_05_jnt, bone2Name=dyn_Jacket_ft_l_04_jnt, swing1LimitDegrees=32, swing2LimitDegrees=32, twistLimitDegrees=12}
 {bone1Name=dyn_Jacket_ft_r_01_jnt, bone2Name=dyn_Jacket_ft_r_MSTR, swing1LimitDegrees=2, swing2LimitDegrees=24, twistLimitDegrees=12}
 {bone1Name=dyn_Jacket_ft_r_02_jnt, bone2Name=dyn_Jacket_ft_r_01_jnt, swing1LimitDegrees=24, swing2LimitDegrees=24, twistLimitDegrees=12}
 {bone1Name=dyn_Jacket_ft_r_03_jnt, bone2Name=dyn_Jacket_ft_r_02_jnt, swing1LimitDegrees=32, swing2LimitDegrees=32, twistLimitDegrees=12}
 {bone1Name=dyn_Jacket_ft_r_04_jnt, bone2Name=dyn_Jacket_ft_r_03_jnt, swing1LimitDegrees=32, swing2LimitDegrees=32, twistLimitDegrees=12}
 {bone1Name=dyn_Jacket_ft_r_05_jnt, bone2Name=dyn_Jacket_ft_r_04_jnt, swing1LimitDegrees=32, swing2LimitDegrees=32, twistLimitDegrees=12}
 {bone1Name=dyn_Jacket_l_01_jnt, bone2Name=dyn_Jacket_l_MSTR, swing1LimitDegrees=12, swing2LimitDegrees=24, twistLimitDegrees=12}
 {bone1Name=dyn_Jacket_l_02_jnt, bone2Name=dyn_Jacket_l_01_jnt, swing1LimitDegrees=12, swing2LimitDegrees=24, twistLimitDegrees=12}
 {bone1Name=dyn_Jacket_l_03_jnt, bone2Name=dyn_Jacket_l_02_jnt, swing1LimitDegrees=12, swing2LimitDegrees=12, twistLimitDegrees=12}
 {bone1Name=dyn_Jacket_l_04_jnt, bone2Name=dyn_Jacket_l_03_jnt, swing1LimitDegrees=12, swing2LimitDegrees=12, twistLimitDegrees=12}
 {bone1Name=dyn_Jacket_l_05_jnt, bone2Name=dyn_Jacket_l_04_jnt, swing1LimitDegrees=12, swing2LimitDegrees=12, twistLimitDegrees=12}
 {bone1Name=dyn_Jacket_l_06_jnt, bone2Name=dyn_Jacket_l_05_jnt, swing1LimitDegrees=12, swing2LimitDegrees=12, twistLimitDegrees=12}
 {bone1Name=dyn_Jacket_l_07_jnt, bone2Name=dyn_Jacket_l_06_jnt, swing1LimitDegrees=12, swing2LimitDegrees=12, twistLimitDegrees=12}
 {bone1Name=dyn_Jacket_r_01_jnt, bone2Name=dyn_Jacket_r_MSTR, swing1LimitDegrees=12, swing2LimitDegrees=24, twistLimitDegrees=12}
 {bone1Name=dyn_Jacket_r_02_jnt, bone2Name=dyn_Jacket_r_01_jnt, swing1LimitDegrees=12, swing2LimitDegrees=24, twistLimitDegrees=12}
 {bone1Name=dyn_Jacket_r_03_jnt, bone2Name=dyn_Jacket_r_02_jnt, swing1LimitDegrees=12, swing2LimitDegrees=12, twistLimitDegrees=12}
 {bone1Name=dyn_Jacket_r_04_jnt, bone2Name=dyn_Jacket_r_03_jnt, swing1LimitDegrees=12, swing2LimitDegrees=12, twistLimitDegrees=12}
 {bone1Name=dyn_Jacket_r_05_jnt, bone2Name=dyn_Jacket_r_04_jnt, swing1LimitDegrees=12, swing2LimitDegrees=12, twistLimitDegrees=12}
 {bone1Name=dyn_Jacket_r_06_jnt, bone2Name=dyn_Jacket_r_05_jnt, swing1LimitDegrees=12, swing2LimitDegrees=12, twistLimitDegrees=12}
 {bone1Name=dyn_Jacket_r_07_jnt, bone2Name=dyn_Jacket_r_06_jnt, swing1LimitDegrees=12, swing2LimitDegrees=12, twistLimitDegrees=12}
 {
  bone1Name=spine_04
  bone2Name=spine_03
  swing1LimitDegrees=45
  swing1Motion=Free
  swing2LimitDegrees=45
  swing2Motion=Free
  twistLimitDegrees=45
  twistMotion=Free
 }
 {
  bone1Name=spine_01
  bone2Name=pelvis
  swing1LimitDegrees=45
  swing1Motion=Free
  swing2LimitDegrees=45
  swing2Motion=Free
  twistLimitDegrees=45
  twistMotion=Free
 }
 {
  bone1Name=spine_03
  bone2Name=spine_02
  swing1LimitDegrees=45
  swing1Motion=Free
  swing2LimitDegrees=45
  swing2Motion=Free
  twistLimitDegrees=45
  twistMotion=Free
 }
 {
  bone1Name=spine_02
  bone2Name=spine_01
  swing1LimitDegrees=45
  swing1Motion=Free
  swing2LimitDegrees=45
  swing2Motion=Free
  twistLimitDegrees=45
  twistMotion=Free
 }
 {
  bone1Name=thigh_l
  bone2Name=pelvis
  swing1LimitDegrees=45
  swing1Motion=Free
  swing2LimitDegrees=45
  swing2Motion=Free
  twistLimitDegrees=45
  twistMotion=Free
 }
 {
  bone1Name=calf_l
  bone2Name=thigh_l
  swing1LimitDegrees=45
  swing1Motion=Free
  swing2LimitDegrees=45
  swing2Motion=Free
  twistLimitDegrees=45
  twistMotion=Free
 }
 {
  bone1Name=thigh_r
  bone2Name=pelvis
  swing1LimitDegrees=45
  swing1Motion=Free
  swing2LimitDegrees=45
  swing2Motion=Free
  twistLimitDegrees=45
  twistMotion=Free
 }
 {
  bone1Name=calf_r
  bone2Name=thigh_r
  swing1LimitDegrees=45
  swing1Motion=Free
  swing2LimitDegrees=45
  swing2Motion=Free
  twistLimitDegrees=45
  twistMotion=Free
 }
 {
  bone1Name=foot_r
  bone2Name=calf_r
  swing1LimitDegrees=45
  swing1Motion=Free
  swing2LimitDegrees=45
  swing2Motion=Free
  twistLimitDegrees=45
  twistMotion=Free
 }
 {
  bone1Name=foot_l
  bone2Name=calf_l
  swing1LimitDegrees=45
  swing1Motion=Free
  swing2LimitDegrees=45
  swing2Motion=Free
  twistLimitDegrees=45
  twistMotion=Free
 }
 {
  bone1Name=dyn_Jacket_ft_r_05_jnt
  bone2Name=dyn_Jacket_r_05_jnt
  swing1LimitDegrees=45
  swing1Motion=Free
  swing2LimitDegrees=45
  swing2Motion=Free
  twistLimitDegrees=45
  twistMotion=Free
 }
]
