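# Anim Blueprint for the PrismParticle male body cosmetic, animating its jacket and name tag over the copied parent pose.
# Bone constraints and a stance-driven Control Rig feed RigidBody simulation, with a constraint-only fallback when rigid body is disabled.
# Skydiving switches to its own physics asset and lift offsets, then AnimDynamics swings the name tag.
# The EventGraph derives gravity override and movement-state alphas that scale coat physics and bone offsets.
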AnimBlueprint /BRCosmetics/Characters/Player/Male/Medium/Bodies/M_MED_PrismParticle/Meshes/M_MED_PrismParticle_AnimBP
target_skeleton=&/M_MED_PrismParticle_Skeleton
graphs [
 AnimGraph {nodes [
   n1 = "Constraint - dyn_Jacket_bk_r_01_jnt (P)"(ComponentPose=Local To Component(LocalPose=Copy Pose From Mesh() {bUseAttachedParent=true}), Alpha=1, ConstraintWeights_0=1) {
    alphaBoolBlend {blendOption=Linear}
    boneToModify {boneName=dyn_Jacket_bk_r_01_jnt}
    constraintSetup [{targetBone {boneName=spine_04}}]
    constraintWeights [1]
   }
   n2 = "Constraint - dyn_Jacket_bk_r_02_jnt (PP)"!(ComponentPose=@n1, Alpha=1, ConstraintWeights_0=0.5, ConstraintWeights_1:float=0.5) -> (Pose) {
    alphaBoolBlend {blendOption=Linear}
    boneToModify {boneName=dyn_Jacket_bk_r_02_jnt}
    constraintSetup [{targetBone {boneName=spine_03}}, {targetBone {boneName=spine_02}}]
    constraintWeights [1,1]
   }
   n3 = "Constraint - dyn_Jacket_bk_l_01_jnt (P)"(ComponentPose=@n2, Alpha=1, ConstraintWeights_0=1) {
    alphaBoolBlend {blendOption=Linear}
    boneToModify {boneName=dyn_Jacket_bk_l_01_jnt}
    constraintSetup [{targetBone {boneName=spine_04}}]
    constraintWeights [1]
   }
   n4 = "Constraint - dyn_Jacket_bk_l_02_jnt (PP)"!(ComponentPose=@n3, Alpha=1, ConstraintWeights_0=0.5, ConstraintWeights_1:float=0.5) -> (Pose) {
    alphaBoolBlend {blendOption=Linear}
    boneToModify {boneName=dyn_Jacket_bk_l_02_jnt}
    constraintSetup [{targetBone {boneName=spine_03}}, {targetBone {boneName=spine_02}}]
    constraintWeights [1,1]
   }
   n5 = "Constraint - dyn_Jacket_bk_c_01_jnt (P)"(ComponentPose=@n4, Alpha=1, ConstraintWeights_0=1) {
    alphaBoolBlend {blendOption=Linear}
    boneToModify {boneName=dyn_Jacket_bk_c_01_jnt}
    constraintSetup [{targetBone {boneName=spine_04}}]
    constraintWeights [1]
   }
   n6 = "Constraint - dyn_Jacket_bk_c_02_jnt (PP)"!(ComponentPose=@n5, Alpha=1, ConstraintWeights_0=0.5, ConstraintWeights_1:float=0.5) -> (Pose) {
    alphaBoolBlend {blendOption=Linear}
    boneToModify {boneName=dyn_Jacket_bk_c_02_jnt}
    constraintSetup [{targetBone {boneName=spine_03}}, {targetBone {boneName=spine_02}}]
    constraintWeights [1,1]
   }
   n7 = "Constraint - dyn_Jacket_r_01_jnt (P)"(ComponentPose=@n6, Alpha=1, ConstraintWeights_0=1) {
    alphaBoolBlend {blendOption=Linear}
    boneToModify {boneName=dyn_Jacket_r_01_jnt}
    constraintSetup [{targetBone {boneName=spine_04}}]
    constraintWeights [1]
   }
   n8 = "Constraint - dyn_Jacket_r_02_jnt (PP)"!(ComponentPose=@n7, Alpha=1, ConstraintWeights_0=0.5, ConstraintWeights_1:float=0.5) -> (Pose) {
    alphaBoolBlend {blendOption=Linear}
    boneToModify {boneName=dyn_Jacket_r_02_jnt}
    constraintSetup [{targetBone {boneName=spine_03}}, {targetBone {boneName=spine_02}}]
    constraintWeights [1,1]
   }
   n9 = "Constraint - dyn_Jacket_l_01_jnt (P)"(ComponentPose=@n8, Alpha=1, ConstraintWeights_0=1) {
    alphaBoolBlend {blendOption=Linear}
    boneToModify {boneName=dyn_Jacket_l_01_jnt}
    constraintSetup [{targetBone {boneName=spine_04}}]
    constraintWeights [1]
   }
   n10 = "Constraint - dyn_Jacket_l_02_jnt (PP)"!(ComponentPose=@n9, Alpha=1, ConstraintWeights_0=0.5, ConstraintWeights_1:float=0.5) -> (Pose) {
    alphaBoolBlend {blendOption=Linear}
    boneToModify {boneName=dyn_Jacket_l_02_jnt}
    constraintSetup [{targetBone {boneName=spine_03}}, {targetBone {boneName=spine_02}}]
    constraintWeights [1,1]
   }
   n11 = "Constraint - dyn_Jacket_ft_r_01_jnt (P)"(ComponentPose=@n10, Alpha=1, ConstraintWeights_0=1) {
    alphaBoolBlend {blendOption=Linear}
    boneToModify {boneName=dyn_Jacket_ft_r_01_jnt}
    constraintSetup [{targetBone {boneName=spine_04}}]
    constraintWeights [1]
   }
   n12 = "Constraint - dyn_Jacket_ft_r_02_jnt (PP)"!(ComponentPose=@n11, Alpha=1, ConstraintWeights_0=0.5, ConstraintWeights_1:float=0.5) -> (Pose) {
    alphaBoolBlend {blendOption=Linear}
    boneToModify {boneName=dyn_Jacket_ft_r_02_jnt}
    constraintSetup [{targetBone {boneName=spine_03}}, {targetBone {boneName=spine_02}}]
    constraintWeights [1,1]
   }
   n13 = "Constraint - dyn_Jacket_ft_l_01_jnt (P)"(ComponentPose=@n12, Alpha=1, ConstraintWeights_0=1) {
    alphaBoolBlend {blendOption=Linear}
    boneToModify {boneName=dyn_Jacket_ft_l_01_jnt}
    constraintSetup [{targetBone {boneName=spine_04}}]
    constraintWeights [1]
   }
   n14 = "Constraint - dyn_Jacket_ft_l_02_jnt (PP)"!(ComponentPose=@n13, Alpha=1, ConstraintWeights_0=0.5, ConstraintWeights_1:float=0.5) -> (Pose) {
    alphaBoolBlend {blendOption=Linear}
    boneToModify {boneName=dyn_Jacket_ft_l_02_jnt}
    constraintSetup [{targetBone {boneName=spine_03}}, {targetBone {boneName=spine_02}}]
    constraintWeights [1,1]
   }
   n15 = Control Rig!(
    Source=Component To Local(ComponentPose=@n14)
    Alpha
    Is Rifle Alpha:double=Get Is Rifle Alpha() -> (Is Rifle Alpha:double)
    Is Pistol Alpha:double=Get Is Pistol Alpha() -> (Is Pistol Alpha:double)
    Is Crouching Alpha:double=Get Is Crouching Alpha() -> (Is Crouching Alpha:double)
   ) -> (Pose) {
    alphaBoolBlend {blendOption=Linear}
    controlRigAssetReference {blueprintRigClass=&/M_MED_PrismParticle_CtrlRig.M_MED_PrismParticle_CtrlRig_C}
   }
   n16 = Get Is Playing Melee Anim Alpha() -> (Is Playing Melee Anim Alpha:double)
   n17 = "Transform (Modify) Bone - Bone: dyn_Jacket_bk_r_MSTR"(ComponentPose=Local To Component(LocalPose=@n15), Alpha=@n16, Translation="0,-1.5,1.5", Rotation="0,0,0", Scale="1,1,1") {
    rotationMode=BMM_Additive
    rotationSpace=BCS_BoneSpace
    translationMode=BMM_Additive
    translationSpace=BCS_BoneSpace
    alphaBoolBlend {blendOption=Linear}
    alphaScaleBiasClamp {bInterpResult=true, interpSpeedDecreasing=6}
    boneToModify {boneName=dyn_Jacket_bk_r_MSTR}
    translation {y=-1.5, z=1.5}
   }
   n18 = "Transform (Modify) Bone - Bone: dyn_Jacket_bk_l_MSTR"(ComponentPose=@n17, Alpha=@n16, Translation="0,1.5,1.5", Rotation="0,0,0", Scale="1,1,1") {
    rotationMode=BMM_Additive
    rotationSpace=BCS_BoneSpace
    translationMode=BMM_Additive
    translationSpace=BCS_BoneSpace
    alphaBoolBlend {blendOption=Linear}
    alphaScaleBiasClamp {bInterpResult=true, interpSpeedDecreasing=6}
    boneToModify {boneName=dyn_Jacket_bk_l_MSTR}
    translation {y=1.5, z=1.5}
   }
   n19 = "Transform (Modify) Bone - Bone: dyn_Jacket_bk_c_MSTR"(ComponentPose=@n18, Alpha=@n16, Translation="0,0,1.5", Rotation="0,0,0", Scale="1,1,1") {
    rotationMode=BMM_Additive
    rotationSpace=BCS_BoneSpace
    translationMode=BMM_Additive
    translationSpace=BCS_BoneSpace
    alphaBoolBlend {blendOption=Linear}
    alphaScaleBiasClamp {bInterpResult=true, interpSpeedDecreasing=6}
    boneToModify {boneName=dyn_Jacket_bk_c_MSTR}
    translation {z=1.5}
   }
   n20 = RigidBody!(ComponentPose=@n19, Alpha=Get Coat Alpha() -> (Coat Alpha:double), OverrideWorldGravity:Vector=Get Gravity Override(), ExternalForce) -> (Pose) {
    bClampLinearTranslationLimitToRefPose=true
    bOverrideWorldGravity=true
    overridePhysicsAsset=&/M_MED_PrismParticle_Physics
    simulationSpace=BaseBoneSpace
    alphaBoolBlend {blendOption=Linear}
    alphaScaleBiasClamp {bInterpResult=true, interpSpeedIncreasing=2}
    baseBoneRef {boneName=spine_03}
    componentAppliedLinearAccClamp=<3000 3000 3000>
    componentLinearAccScale=<0.5 0.5 0.5>
    componentLinearVelScale=<0.5 0.5 0.5>
    simSpaceSettings {
     maxAngularAcceleration=1000
     maxAngularVelocity=3
     maxLinearAcceleration=700
     maxLinearVelocity=500
     velocityScaleZ=0
     worldAlpha=0.5
     externalLinearDragV=<0.2 0.2>
    }
   }
   n21 = Control Rig(
    Source=Copy Pose From Mesh() {bUseAttachedParent=true}
    Is Rifle Alpha=Get Is Rifle Alpha()
    Is Pistol Alpha=Get Is Pistol Alpha()
    Is Crouching Alpha=Get Is Crouching Alpha()
   ) {
    alphaBoolBlend {blendOption=Linear}
    controlRigAssetReference {blueprintRigClass=&M_MED_PrismParticle_CtrlRig_C}
   }
   n22 = "Constraint - dyn_Jacket_bk_r_01_jnt (P)"(ComponentPose=Local To Component(LocalPose=@n21), Alpha=1, ConstraintWeights_0=1) {
    alphaBoolBlend {blendOption=Linear}
    boneToModify {boneName=dyn_Jacket_bk_r_01_jnt}
    constraintSetup [{targetBone {boneName=spine_04}}]
    constraintWeights [1]
   }
   n23 = "Constraint - dyn_Jacket_bk_r_02_jnt (PP)"(ComponentPose=@n22, Alpha=1, ConstraintWeights_0=0.5, ConstraintWeights_1=0.5) {
    alphaBoolBlend {blendOption=Linear}
    boneToModify {boneName=dyn_Jacket_bk_r_02_jnt}
    constraintSetup [{targetBone {boneName=spine_03}}, {targetBone {boneName=spine_02}}]
    constraintWeights [1,1]
   }
   n24 = "Constraint - dyn_Jacket_bk_r_03_jnt (PP)"!(ComponentPose=@n23, Alpha=1, ConstraintWeights_0=0.5, ConstraintWeights_1:float=0.5) -> (Pose) {
    alphaBoolBlend {blendOption=Linear}
    boneToModify {boneName=dyn_Jacket_bk_r_03_jnt}
    constraintSetup [{targetBone {boneName=spine_02}}, {targetBone {boneName=spine_01}}]
    constraintWeights [1,1]
   }
   n25 = "Constraint - dyn_Jacket_bk_r_04_jnt (PP)"!(ComponentPose=@n24, Alpha=1, ConstraintWeights_0=0.5, ConstraintWeights_1:float=0.5) -> (Pose) {
    alphaBoolBlend {blendOption=Linear}
    boneToModify {boneName=dyn_Jacket_bk_r_04_jnt}
    constraintSetup [{targetBone {boneName=pelvis}}, {targetBone {boneName=thigh_r}}]
    constraintWeights [1,1]
   }
   n26 = "Constraint - dyn_Jacket_bk_r_05_jnt (PP)"!(ComponentPose=@n25, Alpha=1, ConstraintWeights_0=0.5, ConstraintWeights_1:float=0.5) -> (Pose) {
    alphaBoolBlend {blendOption=Linear}
    boneToModify {boneName=dyn_Jacket_bk_r_05_jnt}
    constraintSetup [{targetBone {boneName=pelvis}}, {targetBone {boneName=thigh_r}}]
    constraintWeights [1,1]
   }
   n27 = "Constraint - dyn_Jacket_bk_r_06_jnt (PP)"!(ComponentPose=@n26, Alpha=1, ConstraintWeights_0=0.8, ConstraintWeights_1:float=0.2) -> (Pose) {
    alphaBoolBlend {blendOption=Linear}
    boneToModify {boneName=dyn_Jacket_bk_r_06_jnt}
    constraintSetup [{targetBone {boneName=pelvis}}, {targetBone {boneName=thigh_r}}]
    constraintWeights [1,1]
   }
   n28 = "Constraint - dyn_Jacket_bk_r_07_jnt (PP)"!(ComponentPose=@n27, Alpha=1, ConstraintWeights_0=0.8, ConstraintWeights_1:float=0.2) -> (Pose) {
    alphaBoolBlend {blendOption=Linear}
    boneToModify {boneName=dyn_Jacket_bk_r_07_jnt}
    constraintSetup [{targetBone {boneName=pelvis}}, {targetBone {boneName=thigh_r}}]
    constraintWeights [1,1]
   }
   n29 = "Constraint - dyn_Jacket_bk_l_01_jnt (P)"(ComponentPose=@n28, Alpha=1, ConstraintWeights_0=1) {
    alphaBoolBlend {blendOption=Linear}
    boneToModify {boneName=dyn_Jacket_bk_l_01_jnt}
    constraintSetup [{targetBone {boneName=spine_04}}]
    constraintWeights [1]
   }
   n30 = "Constraint - dyn_Jacket_bk_l_02_jnt (PP)"(ComponentPose=@n29, Alpha=1, ConstraintWeights_0=0.5, ConstraintWeights_1=0.5) {
    alphaBoolBlend {blendOption=Linear}
    boneToModify {boneName=dyn_Jacket_bk_l_02_jnt}
    constraintSetup [{targetBone {boneName=spine_03}}, {targetBone {boneName=spine_02}}]
    constraintWeights [1,1]
   }
   n31 = "Constraint - dyn_Jacket_bk_l_03_jnt (PP)"!(ComponentPose=@n30, Alpha=1, ConstraintWeights_0=0.5, ConstraintWeights_1:float=0.5) -> (Pose) {
    alphaBoolBlend {blendOption=Linear}
    boneToModify {boneName=dyn_Jacket_bk_l_03_jnt}
    constraintSetup [{targetBone {boneName=spine_02}}, {targetBone {boneName=spine_01}}]
    constraintWeights [1,1]
   }
   n32 = "Constraint - dyn_Jacket_bk_l_04_jnt (PP)"!(ComponentPose=@n31, Alpha=1, ConstraintWeights_0=0.5, ConstraintWeights_1:float=0.5) -> (Pose) {
    alphaBoolBlend {blendOption=Linear}
    boneToModify {boneName=dyn_Jacket_bk_l_04_jnt}
    constraintSetup [{targetBone {boneName=pelvis}}, {targetBone {boneName=thigh_l}}]
    constraintWeights [1,1]
   }
   n33 = "Constraint - dyn_Jacket_bk_l_05_jnt (PP)"!(ComponentPose=@n32, Alpha=1, ConstraintWeights_0=0.5, ConstraintWeights_1:float=0.5) -> (Pose) {
    alphaBoolBlend {blendOption=Linear}
    boneToModify {boneName=dyn_Jacket_bk_l_05_jnt}
    constraintSetup [{targetBone {boneName=pelvis}}, {targetBone {boneName=thigh_l}}]
    constraintWeights [1,1]
   }
   n34 = "Constraint - dyn_Jacket_bk_l_06_jnt (PP)"!(ComponentPose=@n33, Alpha=1, ConstraintWeights_0=0.8, ConstraintWeights_1:float=0.2) -> (Pose) {
    alphaBoolBlend {blendOption=Linear}
    boneToModify {boneName=dyn_Jacket_bk_l_06_jnt}
    constraintSetup [{targetBone {boneName=pelvis}}, {targetBone {boneName=thigh_l}}]
    constraintWeights [1,1]
   }
   n35 = "Constraint - dyn_Jacket_bk_l_07_jnt (PP)"!(ComponentPose=@n34, Alpha=1, ConstraintWeights_0=0.8, ConstraintWeights_1:float=0.2) -> (Pose) {
    alphaBoolBlend {blendOption=Linear}
    boneToModify {boneName=dyn_Jacket_bk_l_07_jnt}
    constraintSetup [{targetBone {boneName=pelvis}}, {targetBone {boneName=thigh_l}}]
    constraintWeights [1,1]
   }
   n36 = "Constraint - dyn_Jacket_bk_c_01_jnt (P)"(ComponentPose=@n35, Alpha=1, ConstraintWeights_0=1) {
    alphaBoolBlend {blendOption=Linear}
    boneToModify {boneName=dyn_Jacket_bk_c_01_jnt}
    constraintSetup [{targetBone {boneName=spine_04}}]
    constraintWeights [1]
   }
   n37 = "Constraint - dyn_Jacket_bk_c_02_jnt (PP)"(ComponentPose=@n36, Alpha=1, ConstraintWeights_0=0.5, ConstraintWeights_1=0.5) {
    alphaBoolBlend {blendOption=Linear}
    boneToModify {boneName=dyn_Jacket_bk_c_02_jnt}
    constraintSetup [{targetBone {boneName=spine_03}}, {targetBone {boneName=spine_02}}]
    constraintWeights [1,1]
   }
   n38 = "Constraint - dyn_Jacket_bk_c_03_jnt (PP)"!(ComponentPose=@n37, Alpha=1, ConstraintWeights_0=0.5, ConstraintWeights_1:float=0.5) -> (Pose) {
    alphaBoolBlend {blendOption=Linear}
    boneToModify {boneName=dyn_Jacket_bk_c_03_jnt}
    constraintSetup [{targetBone {boneName=spine_02}}, {targetBone {boneName=spine_01}}]
    constraintWeights [1,1]
   }
   n39 = "Constraint - dyn_Jacket_bk_c_04_jnt (P)"(ComponentPose=@n38, Alpha=1, ConstraintWeights_0=1) {
    alphaBoolBlend {blendOption=Linear}
    boneToModify {boneName=dyn_Jacket_bk_c_04_jnt}
    constraintSetup [{targetBone {boneName=pelvis}}]
    constraintWeights [1]
   }
   n40 = "Constraint - dyn_Jacket_bk_c_05_jnt (P)"(ComponentPose=@n39, Alpha=1, ConstraintWeights_0=1) {
    alphaBoolBlend {blendOption=Linear}
    boneToModify {boneName=dyn_Jacket_bk_c_05_jnt}
    constraintSetup [{targetBone {boneName=pelvis}}]
    constraintWeights [1]
   }
   n41 = "Constraint - dyn_Jacket_bk_c_06_jnt (P)"(ComponentPose=@n40, Alpha=1, ConstraintWeights_0=1) {
    alphaBoolBlend {blendOption=Linear}
    boneToModify {boneName=dyn_Jacket_bk_c_06_jnt}
    constraintSetup [{targetBone {boneName=pelvis}}]
    constraintWeights [1]
   }
   n42 = "Constraint - dyn_Jacket_r_01_jnt (P)"(ComponentPose=@n41, Alpha=1, ConstraintWeights_0=1) {
    alphaBoolBlend {blendOption=Linear}
    boneToModify {boneName=dyn_Jacket_r_01_jnt}
    constraintSetup [{targetBone {boneName=spine_04}}]
    constraintWeights [1]
   }
   n43 = "Constraint - dyn_Jacket_r_02_jnt (PP)"(ComponentPose=@n42, Alpha=1, ConstraintWeights_0=0.5, ConstraintWeights_1=0.5) {
    alphaBoolBlend {blendOption=Linear}
    boneToModify {boneName=dyn_Jacket_r_02_jnt}
    constraintSetup [{targetBone {boneName=spine_03}}, {targetBone {boneName=spine_02}}]
    constraintWeights [1,1]
   }
   n44 = "Constraint - dyn_Jacket_r_03_jnt (PP)"!(ComponentPose=@n43, Alpha=1, ConstraintWeights_0=0.5, ConstraintWeights_1:float=0.5) -> (Pose) {
    alphaBoolBlend {blendOption=Linear}
    boneToModify {boneName=dyn_Jacket_r_03_jnt}
    constraintSetup [{targetBone {boneName=spine_02}}, {targetBone {boneName=spine_01}}]
    constraintWeights [1,1]
   }
   n45 = "Constraint - dyn_Jacket_r_04_jnt (PP)"!(ComponentPose=@n44, Alpha=1, ConstraintWeights_0=0.3, ConstraintWeights_1:float=0.7) -> (Pose) {
    alphaBoolBlend {blendOption=Linear}
    boneToModify {boneName=dyn_Jacket_r_04_jnt}
    constraintSetup [{targetBone {boneName=pelvis}}, {targetBone {boneName=thigh_r}}]
    constraintWeights [1,1]
   }
   n46 = "Constraint - dyn_Jacket_r_05_jnt (PP)"!(ComponentPose=@n45, Alpha=1, ConstraintWeights_0=0.2, ConstraintWeights_1:float=0.8) -> (Pose) {
    alphaBoolBlend {blendOption=Linear}
    boneToModify {boneName=dyn_Jacket_r_05_jnt}
    constraintSetup [{targetBone {boneName=pelvis}}, {targetBone {boneName=thigh_r}}]
    constraintWeights [1,1]
   }
   n47 = "Constraint - dyn_Jacket_r_06_jnt (PP)"!(ComponentPose=@n46, Alpha=1, ConstraintWeights_0=0.2, ConstraintWeights_1:float=0.8) -> (Pose) {
    alphaBoolBlend {blendOption=Linear}
    boneToModify {boneName=dyn_Jacket_r_06_jnt}
    constraintSetup [{targetBone {boneName=pelvis}}, {targetBone {boneName=thigh_r}}]
    constraintWeights [1,1]
   }
   n48 = "Constraint - dyn_Jacket_r_07_jnt (PP)"!(ComponentPose=@n47, Alpha=1, ConstraintWeights_0=0.2, ConstraintWeights_1:float=0.8) -> (Pose) {
    alphaBoolBlend {blendOption=Linear}
    boneToModify {boneName=dyn_Jacket_r_07_jnt}
    constraintSetup [{targetBone {boneName=pelvis}}, {targetBone {boneName=thigh_r}}]
    constraintWeights [1,1]
   }
   n49 = "Constraint - dyn_Jacket_l_01_jnt (P)"(ComponentPose=@n48, Alpha=1, ConstraintWeights_0=1) {
    alphaBoolBlend {blendOption=Linear}
    boneToModify {boneName=dyn_Jacket_l_01_jnt}
    constraintSetup [{targetBone {boneName=spine_04}}]
    constraintWeights [1]
   }
   n50 = "Constraint - dyn_Jacket_l_02_jnt (PP)"(ComponentPose=@n49, Alpha=1, ConstraintWeights_0=0.5, ConstraintWeights_1=0.5) {
    alphaBoolBlend {blendOption=Linear}
    boneToModify {boneName=dyn_Jacket_l_02_jnt}
    constraintSetup [{targetBone {boneName=spine_03}}, {targetBone {boneName=spine_02}}]
    constraintWeights [1,1]
   }
   n51 = "Constraint - dyn_Jacket_l_03_jnt (PP)"!(ComponentPose=@n50, Alpha=1, ConstraintWeights_0=0.5, ConstraintWeights_1:float=0.5) -> (Pose) {
    alphaBoolBlend {blendOption=Linear}
    boneToModify {boneName=dyn_Jacket_l_03_jnt}
    constraintSetup [{targetBone {boneName=spine_02}}, {targetBone {boneName=spine_01}}]
    constraintWeights [1,1]
   }
   n52 = "Constraint - dyn_Jacket_l_04_jnt (PP)"!(ComponentPose=@n51, Alpha=1, ConstraintWeights_0=0.3, ConstraintWeights_1:float=0.7) -> (Pose) {
    alphaBoolBlend {blendOption=Linear}
    boneToModify {boneName=dyn_Jacket_l_04_jnt}
    constraintSetup [{targetBone {boneName=pelvis}}, {targetBone {boneName=thigh_l}}]
    constraintWeights [1,1]
   }
   n53 = "Constraint - dyn_Jacket_l_05_jnt (PP)"!(ComponentPose=@n52, Alpha=1, ConstraintWeights_0=0.2, ConstraintWeights_1:float=0.8) -> (Pose) {
    alphaBoolBlend {blendOption=Linear}
    boneToModify {boneName=dyn_Jacket_l_05_jnt}
    constraintSetup [{targetBone {boneName=pelvis}}, {targetBone {boneName=thigh_l}}]
    constraintWeights [1,1]
   }
   n54 = "Constraint - dyn_Jacket_l_06_jnt (PP)"!(ComponentPose=@n53, Alpha=1, ConstraintWeights_0=0.2, ConstraintWeights_1:float=0.8) -> (Pose) {
    alphaBoolBlend {blendOption=Linear}
    boneToModify {boneName=dyn_Jacket_l_06_jnt}
    constraintSetup [{targetBone {boneName=pelvis}}, {targetBone {boneName=thigh_l}}]
    constraintWeights [1,1]
   }
   n55 = "Constraint - dyn_Jacket_l_07_jnt (PP)"!(ComponentPose=@n54, Alpha=1, ConstraintWeights_0=0.2, ConstraintWeights_1:float=0.8) -> (Pose) {
    alphaBoolBlend {blendOption=Linear}
    boneToModify {boneName=dyn_Jacket_l_07_jnt}
    constraintSetup [{targetBone {boneName=pelvis}}, {targetBone {boneName=thigh_l}}]
    constraintWeights [1,1]
   }
   n56 = "Constraint - dyn_Jacket_ft_r_01_jnt (P)"(ComponentPose=@n55, Alpha=1, ConstraintWeights_0=1) {
    alphaBoolBlend {blendOption=Linear}
    boneToModify {boneName=dyn_Jacket_ft_r_01_jnt}
    constraintSetup [{targetBone {boneName=spine_04}}]
    constraintWeights [1]
   }
   n57 = "Constraint - dyn_Jacket_ft_r_02_jnt (PP)"(ComponentPose=@n56, Alpha=1, ConstraintWeights_0=0.8, ConstraintWeights_1=0.2) {
    alphaBoolBlend {blendOption=Linear}
    boneToModify {boneName=dyn_Jacket_ft_r_02_jnt}
    constraintSetup [{targetBone {boneName=spine_03}}, {targetBone {boneName=thigh_r}}]
    constraintWeights [1,1]
   }
   n58 = "Constraint - dyn_Jacket_ft_r_03_jnt (PP)"!(ComponentPose=@n57, Alpha=1, ConstraintWeights_0=0.8, ConstraintWeights_1:float=0.2) -> (Pose) {
    alphaBoolBlend {blendOption=Linear}
    boneToModify {boneName=dyn_Jacket_ft_r_03_jnt}
    constraintSetup [{targetBone {boneName=spine_02}}, {targetBone {boneName=thigh_r}}]
    constraintWeights [1,1]
   }
   n59 = "Constraint - dyn_Jacket_ft_r_04_jnt (PP)"!(ComponentPose=@n58, Alpha=1, ConstraintWeights_0=0.35, ConstraintWeights_1:float=0.65) -> (Pose) {
    alphaBoolBlend {blendOption=Linear}
    boneToModify {boneName=dyn_Jacket_ft_r_04_jnt}
    constraintSetup [{targetBone {boneName=pelvis}}, {targetBone {boneName=thigh_r}}]
    constraintWeights [1,1]
   }
   n60 = "Constraint - dyn_Jacket_ft_r_05_jnt (PP)"!(ComponentPose=@n59, Alpha=1, ConstraintWeights_0=0.2, ConstraintWeights_1:float=0.8) -> (Pose) {
    alphaBoolBlend {blendOption=Linear}
    boneToModify {boneName=dyn_Jacket_ft_r_05_jnt}
    constraintSetup [{targetBone {boneName=pelvis}}, {targetBone {boneName=thigh_r}}]
    constraintWeights [1,1]
   }
   n61 = "Constraint - dyn_Jacket_ft_l_01_jnt (P)"(ComponentPose=@n60, Alpha=1, ConstraintWeights_0=1) {
    alphaBoolBlend {blendOption=Linear}
    boneToModify {boneName=dyn_Jacket_ft_l_01_jnt}
    constraintSetup [{targetBone {boneName=spine_04}}]
    constraintWeights [1]
   }
   n62 = "Constraint - dyn_Jacket_ft_l_02_jnt (PP)"(ComponentPose=@n61, Alpha=1, ConstraintWeights_0=0.8, ConstraintWeights_1=0.2) {
    alphaBoolBlend {blendOption=Linear}
    boneToModify {boneName=dyn_Jacket_ft_l_02_jnt}
    constraintSetup [{targetBone {boneName=spine_03}}, {targetBone {boneName=thigh_l}}]
    constraintWeights [1,1]
   }
   n63 = "Constraint - dyn_Jacket_ft_l_03_jnt (PP)"!(ComponentPose=@n62, Alpha=1, ConstraintWeights_0=0.8, ConstraintWeights_1:float=0.2) -> (Pose) {
    alphaBoolBlend {blendOption=Linear}
    boneToModify {boneName=dyn_Jacket_ft_l_03_jnt}
    constraintSetup [{targetBone {boneName=spine_02}}, {targetBone {boneName=thigh_l}}]
    constraintWeights [1,1]
   }
   n64 = "Constraint - dyn_Jacket_ft_l_04_jnt (PP)"!(ComponentPose=@n63, Alpha=1, ConstraintWeights_0=0.35, ConstraintWeights_1:float=0.65) -> (Pose) {
    alphaBoolBlend {blendOption=Linear}
    boneToModify {boneName=dyn_Jacket_ft_l_04_jnt}
    constraintSetup [{targetBone {boneName=pelvis}}, {targetBone {boneName=thigh_l}}]
    constraintWeights [1,1]
   }
   n65 = "Constraint - dyn_Jacket_ft_l_05_jnt (PP)"!(ComponentPose=@n64, Alpha=1, ConstraintWeights_0=0.2, ConstraintWeights_1:float=0.8) -> (Pose) {
    alphaBoolBlend {blendOption=Linear}
    boneToModify {boneName=dyn_Jacket_ft_l_05_jnt}
    constraintSetup [{targetBone {boneName=pelvis}}, {targetBone {boneName=thigh_l}}]
    constraintWeights [1,1]
   }
   n66 = Blend Poses by bool(BlendPose_0=Component To Local(ComponentPose=@n20), BlendPose_1=Component To Local(ComponentPose=@n65), bActiveValue=Get bRigidBodyEnabled()) {blendType=Linear}
   n67 = Control Rig(
    Source=Copy Pose From Mesh() {bUseAttachedParent=true}
    Is Rifle Alpha=Get Is Rifle Alpha()
    Is Pistol Alpha=Get Is Pistol Alpha()
    Is Crouching Alpha=Get Is Crouching Alpha()
   ) {
    alphaBoolBlend {blendOption=Linear}
    controlRigAssetReference {blueprintRigClass=&M_MED_PrismParticle_CtrlRig_C}
   }
   n68 = "Constraint - dyn_Jacket_bk_r_01_jnt (P)"(ComponentPose=Local To Component(LocalPose=@n67), Alpha=1, ConstraintWeights_0=1) {
    alphaBoolBlend {blendOption=Linear}
    boneToModify {boneName=dyn_Jacket_bk_r_01_jnt}
    constraintSetup [{targetBone {boneName=spine_04}}]
    constraintWeights [1]
   }
   n69 = "Constraint - dyn_Jacket_bk_r_02_jnt (PP)"(ComponentPose=@n68, Alpha=1, ConstraintWeights_0=0.5, ConstraintWeights_1=0.5) {
    alphaBoolBlend {blendOption=Linear}
    boneToModify {boneName=dyn_Jacket_bk_r_02_jnt}
    constraintSetup [{targetBone {boneName=spine_03}}, {targetBone {boneName=spine_02}}]
    constraintWeights [1,1]
   }
   n70 = "Constraint - dyn_Jacket_bk_r_03_jnt (PP)"(ComponentPose=@n69, Alpha=1, ConstraintWeights_0=0.5, ConstraintWeights_1=0.5) {
    alphaBoolBlend {blendOption=Linear}
    boneToModify {boneName=dyn_Jacket_bk_r_03_jnt}
    constraintSetup [{targetBone {boneName=spine_02}}, {targetBone {boneName=spine_01}}]
    constraintWeights [1,1]
   }
   n71 = "Constraint - dyn_Jacket_bk_l_01_jnt (P)"(ComponentPose=@n70, Alpha=1, ConstraintWeights_0=1) {
    alphaBoolBlend {blendOption=Linear}
    boneToModify {boneName=dyn_Jacket_bk_l_01_jnt}
    constraintSetup [{targetBone {boneName=spine_04}}]
    constraintWeights [1]
   }
   n72 = "Constraint - dyn_Jacket_bk_l_02_jnt (PP)"(ComponentPose=@n71, Alpha=1, ConstraintWeights_0=0.5, ConstraintWeights_1=0.5) {
    alphaBoolBlend {blendOption=Linear}
    boneToModify {boneName=dyn_Jacket_bk_l_02_jnt}
    constraintSetup [{targetBone {boneName=spine_03}}, {targetBone {boneName=spine_02}}]
    constraintWeights [1,1]
   }
   n73 = "Constraint - dyn_Jacket_bk_l_03_jnt (PP)"(ComponentPose=@n72, Alpha=1, ConstraintWeights_0=0.5, ConstraintWeights_1=0.5) {
    alphaBoolBlend {blendOption=Linear}
    boneToModify {boneName=dyn_Jacket_bk_l_03_jnt}
    constraintSetup [{targetBone {boneName=spine_02}}, {targetBone {boneName=spine_01}}]
    constraintWeights [1,1]
   }
   n74 = "Constraint - dyn_Jacket_bk_c_01_jnt (P)"(ComponentPose=@n73, Alpha=1, ConstraintWeights_0=1) {
    alphaBoolBlend {blendOption=Linear}
    boneToModify {boneName=dyn_Jacket_bk_c_01_jnt}
    constraintSetup [{targetBone {boneName=spine_04}}]
    constraintWeights [1]
   }
   n75 = "Constraint - dyn_Jacket_bk_c_02_jnt (PP)"(ComponentPose=@n74, Alpha=1, ConstraintWeights_0=0.5, ConstraintWeights_1=0.5) {
    alphaBoolBlend {blendOption=Linear}
    boneToModify {boneName=dyn_Jacket_bk_c_02_jnt}
    constraintSetup [{targetBone {boneName=spine_03}}, {targetBone {boneName=spine_02}}]
    constraintWeights [1,1]
   }
   n76 = "Constraint - dyn_Jacket_bk_c_03_jnt (PP)"(ComponentPose=@n75, Alpha=1, ConstraintWeights_0=0.5, ConstraintWeights_1=0.5) {
    alphaBoolBlend {blendOption=Linear}
    boneToModify {boneName=dyn_Jacket_bk_c_03_jnt}
    constraintSetup [{targetBone {boneName=spine_02}}, {targetBone {boneName=spine_01}}]
    constraintWeights [1,1]
   }
   n77 = "Constraint - dyn_Jacket_r_01_jnt (P)"(ComponentPose=@n76, Alpha=1, ConstraintWeights_0=1) {
    alphaBoolBlend {blendOption=Linear}
    boneToModify {boneName=dyn_Jacket_r_01_jnt}
    constraintSetup [{targetBone {boneName=spine_04}}]
    constraintWeights [1]
   }
   n78 = "Constraint - dyn_Jacket_r_02_jnt (PP)"(ComponentPose=@n77, Alpha=1, ConstraintWeights_0=0.5, ConstraintWeights_1=0.5) {
    alphaBoolBlend {blendOption=Linear}
    boneToModify {boneName=dyn_Jacket_r_02_jnt}
    constraintSetup [{targetBone {boneName=spine_03}}, {targetBone {boneName=spine_02}}]
    constraintWeights [1,1]
   }
   n79 = "Constraint - dyn_Jacket_r_03_jnt (PP)"(ComponentPose=@n78, Alpha=1, ConstraintWeights_0=0.5, ConstraintWeights_1=0.5) {
    alphaBoolBlend {blendOption=Linear}
    boneToModify {boneName=dyn_Jacket_r_03_jnt}
    constraintSetup [{targetBone {boneName=spine_02}}, {targetBone {boneName=spine_01}}]
    constraintWeights [1,1]
   }
   n80 = "Constraint - dyn_Jacket_l_01_jnt (P)"(ComponentPose=@n79, Alpha=1, ConstraintWeights_0=1) {
    alphaBoolBlend {blendOption=Linear}
    boneToModify {boneName=dyn_Jacket_l_01_jnt}
    constraintSetup [{targetBone {boneName=spine_04}}]
    constraintWeights [1]
   }
   n81 = "Constraint - dyn_Jacket_l_02_jnt (PP)"(ComponentPose=@n80, Alpha=1, ConstraintWeights_0=0.5, ConstraintWeights_1=0.5) {
    alphaBoolBlend {blendOption=Linear}
    boneToModify {boneName=dyn_Jacket_l_02_jnt}
    constraintSetup [{targetBone {boneName=spine_03}}, {targetBone {boneName=spine_02}}]
    constraintWeights [1,1]
   }
   n82 = "Constraint - dyn_Jacket_l_03_jnt (PP)"(ComponentPose=@n81, Alpha=1, ConstraintWeights_0=0.5, ConstraintWeights_1=0.5) {
    alphaBoolBlend {blendOption=Linear}
    boneToModify {boneName=dyn_Jacket_l_03_jnt}
    constraintSetup [{targetBone {boneName=spine_02}}, {targetBone {boneName=spine_01}}]
    constraintWeights [1,1]
   }
   n83 = "Constraint - dyn_Jacket_ft_r_01_jnt (P)"(ComponentPose=@n82, Alpha=1, ConstraintWeights_0=1) {
    alphaBoolBlend {blendOption=Linear}
    boneToModify {boneName=dyn_Jacket_ft_r_01_jnt}
    constraintSetup [{targetBone {boneName=spine_04}}]
    constraintWeights [1]
   }
   n84 = "Constraint - dyn_Jacket_ft_r_02_jnt (PP)"(ComponentPose=@n83, Alpha=1, ConstraintWeights_0=0.5, ConstraintWeights_1=0.5) {
    alphaBoolBlend {blendOption=Linear}
    boneToModify {boneName=dyn_Jacket_ft_r_02_jnt}
    constraintSetup [{targetBone {boneName=spine_03}}, {targetBone {boneName=spine_02}}]
    constraintWeights [1,1]
   }
   n85 = "Constraint - dyn_Jacket_ft_r_03_jnt (PP)"(ComponentPose=@n84, Alpha=1, ConstraintWeights_0=0.5, ConstraintWeights_1=0.5) {
    alphaBoolBlend {blendOption=Linear}
    boneToModify {boneName=dyn_Jacket_ft_r_03_jnt}
    constraintSetup [{targetBone {boneName=spine_02}}, {targetBone {boneName=spine_01}}]
    constraintWeights [1,1]
   }
   n86 = "Constraint - dyn_Jacket_ft_l_01_jnt (P)"(ComponentPose=@n85, Alpha=1, ConstraintWeights_0=1) {
    alphaBoolBlend {blendOption=Linear}
    boneToModify {boneName=dyn_Jacket_ft_l_01_jnt}
    constraintSetup [{targetBone {boneName=spine_04}}]
    constraintWeights [1]
   }
   n87 = "Constraint - dyn_Jacket_ft_l_02_jnt (PP)"(ComponentPose=@n86, Alpha=1, ConstraintWeights_0=0.5, ConstraintWeights_1=0.5) {
    alphaBoolBlend {blendOption=Linear}
    boneToModify {boneName=dyn_Jacket_ft_l_02_jnt}
    constraintSetup [{targetBone {boneName=spine_03}}, {targetBone {boneName=spine_02}}]
    constraintWeights [1,1]
   }
   n88 = "Constraint - dyn_Jacket_ft_l_03_jnt (PP)"(ComponentPose=@n87, Alpha=1, ConstraintWeights_0=0.5, ConstraintWeights_1=0.5) {
    alphaBoolBlend {blendOption=Linear}
    boneToModify {boneName=dyn_Jacket_ft_l_03_jnt}
    constraintSetup [{targetBone {boneName=spine_02}}, {targetBone {boneName=spine_01}}]
    constraintWeights [1,1]
   }
   n89 = RigidBody!(
    ComponentPose=@n88
    Alpha
    OverrideWorldGravity=Get Gravity Override()
    ExternalForce
    ComponentLinearAccScale:Vector="0.2,0.2,0.2"
    ComponentLinearVelScale:Vector="0.4,0.3,0.3"
   ) -> (Pose) {
    bClampLinearTranslationLimitToRefPose=true
    bOverrideWorldGravity=true
    overridePhysicsAsset=&/M_MED_PrismParticle_Skydive_Physics
    simulationSpace=BaseBoneSpace
    alphaBoolBlend {blendOption=Linear}
    alphaScaleBiasClamp {bInterpResult=true, interpSpeedIncreasing=2}
    baseBoneRef {boneName=spine_03}
    componentAppliedLinearAccClamp=<3000 3000 3000>
    componentLinearAccScale=<0.4 0.3 0.3>
    componentLinearVelScale=<0.4 0.3 0.3>
    simSpaceSettings {
     maxAngularAcceleration=1000
     maxAngularVelocity=3
     maxLinearAcceleration=700
     maxLinearVelocity=500
     velocityScaleZ=0
     worldAlpha=0.25
     externalLinearDragV=<0.2 0.2>
    }
   }
   n90 = "Transform (Modify) Bone - Bone: dyn_Jacket_bk_r_MSTR"(ComponentPose=@n89, Alpha=1, Translation="0,0,5", Rotation="5,0,0", Scale="1,1,1") {
    rotationMode=BMM_Additive
    rotationSpace=BCS_BoneSpace
    translationMode=BMM_Additive
    translationSpace=BCS_BoneSpace
    alphaBoolBlend {blendOption=Linear}
    boneToModify {boneName=dyn_Jacket_bk_r_MSTR}
    rotation {pitch=5}
    translation {z=5}
   }
   n91 = "Transform (Modify) Bone - Bone: dyn_Jacket_bk_l_MSTR"(ComponentPose=@n90, Alpha=1, Translation="0,0,5", Rotation="5,0,0", Scale="1,1,1") {
    rotationMode=BMM_Additive
    rotationSpace=BCS_BoneSpace
    translationMode=BMM_Additive
    translationSpace=BCS_BoneSpace
    alphaBoolBlend {blendOption=Linear}
    boneToModify {boneName=dyn_Jacket_bk_l_MSTR}
    rotation {pitch=5}
    translation {z=5}
   }
   n92 = "Transform (Modify) Bone - Bone: dyn_Jacket_bk_c_MSTR"(ComponentPose=@n91, Alpha=1, Translation="0,0,5", Rotation="5,0,0", Scale="1,1,1") {
    rotationMode=BMM_Additive
    rotationSpace=BCS_BoneSpace
    translationMode=BMM_Additive
    translationSpace=BCS_BoneSpace
    alphaBoolBlend {blendOption=Linear}
    boneToModify {boneName=dyn_Jacket_bk_c_MSTR}
    rotation {pitch=5}
    translation {z=5}
   }
   n93 = "Transform (Modify) Bone - Bone: dyn_Jacket_r_MSTR"(ComponentPose=@n92, Alpha=1, Translation="0,0,1", Rotation="5,0,0", Scale="1,1,1") {
    rotationMode=BMM_Additive
    rotationSpace=BCS_BoneSpace
    translationMode=BMM_Additive
    translationSpace=BCS_BoneSpace
    alphaBoolBlend {blendOption=Linear}
    boneToModify {boneName=dyn_Jacket_r_MSTR}
    rotation {pitch=5}
    translation {z=1}
   }
   n94 = "Transform (Modify) Bone - Bone: dyn_Jacket_l_MSTR"(ComponentPose=@n93, Alpha=1, Translation="0,0,1", Rotation="5,0,0", Scale="1,1,1") {
    rotationMode=BMM_Additive
    rotationSpace=BCS_BoneSpace
    translationMode=BMM_Additive
    translationSpace=BCS_BoneSpace
    alphaBoolBlend {blendOption=Linear}
    boneToModify {boneName=dyn_Jacket_l_MSTR}
    rotation {pitch=5}
    translation {z=1}
   }
   n95 = Get Is Parachute Opened Float() -> (Is Parachute Opened Float:double)
   n96 = "Transform (Modify) Bone - Bone: dyn_Jacket_bk_r_MSTR"(ComponentPose=@n94, Alpha=@n95, Translation="0,0,0", Rotation="7.5,0,0", Scale="1,1,1") {
    rotationMode=BMM_Additive
    rotationSpace=BCS_BoneSpace
    translationMode=BMM_Additive
    translationSpace=BCS_BoneSpace
    alphaBoolBlend {blendOption=Linear}
    alphaScaleBiasClamp {bInterpResult=true, interpSpeedDecreasing=6}
    boneToModify {boneName=dyn_Jacket_bk_r_MSTR}
    rotation {pitch=7.5}
   }
   n97 = "Transform (Modify) Bone - Bone: dyn_Jacket_bk_l_MSTR"(ComponentPose=@n96, Alpha=@n95, Translation="0,0,0", Rotation="7.5,0,0", Scale="1,1,1") {
    rotationMode=BMM_Additive
    rotationSpace=BCS_BoneSpace
    translationMode=BMM_Additive
    translationSpace=BCS_BoneSpace
    alphaBoolBlend {blendOption=Linear}
    alphaScaleBiasClamp {bInterpResult=true, interpSpeedDecreasing=6}
    boneToModify {boneName=dyn_Jacket_bk_l_MSTR}
    rotation {pitch=7.5}
   }
   n98 = "Transform (Modify) Bone - Bone: dyn_Jacket_bk_c_MSTR"(ComponentPose=@n97, Alpha=@n95, Translation="0,0,0", Rotation="7.5,0,0", Scale="1,1,1") {
    rotationMode=BMM_Additive
    rotationSpace=BCS_BoneSpace
    translationMode=BMM_Additive
    translationSpace=BCS_BoneSpace
    alphaBoolBlend {blendOption=Linear}
    alphaScaleBiasClamp {bInterpResult=true, interpSpeedDecreasing=6}
    boneToModify {boneName=dyn_Jacket_bk_c_MSTR}
    rotation {pitch=7.5}
   }
   n99 = Component To Local(ComponentPose=@n98)
   n100 = Get bIsSkyDiving()
   n101 = "Constraint - dyn_NameTag_MSTR (PPPP)"!(
    ComponentPose=Local To Component(LocalPose=Blend Poses by bool(BlendPose_0=@n99, BlendPose_1=@n66, bActiveValue=@n100) {blendType=Linear})
    Alpha=1
    ConstraintWeights_0=0.4
    ConstraintWeights_1:float=0.2
    ConstraintWeights_2:float=0.2
    ConstraintWeights_3:float=0.2
   ) -> (Pose) {
    alphaBoolBlend {blendOption=Linear}
    boneToModify {boneName=dyn_NameTag_MSTR}
    constraintSetup [
     {targetBone {boneName=deform_pec_base_l}}
     {targetBone {boneName=clavicle_l}}
     {targetBone {boneName=spine_05}}
     {targetBone {boneName=spine_04}}
    ]
    constraintWeights [1,1,1,1]
   }
   n102 = "Transform (Modify) Bone - Bone: dyn_NameTag_01_jnt"(ComponentPose=@n101, Alpha=1, Translation="0,0,0", Rotation="10,0,0", Scale="1,1,1") {
    rotationMode=BMM_Additive
    rotationSpace=BCS_BoneSpace
    translationMode=BMM_Additive
    translationSpace=BCS_BoneSpace
    alphaBoolBlend {blendOption=Linear}
    boneToModify {boneName=dyn_NameTag_01_jnt}
    rotation {pitch=10}
   }
   n103 = Anim Dynamics - Bone: dyn_NameTag_02_jnt(ComponentPose=@n102, Alpha=1, ExternalForce="0,0,0") {
    angularDampingOverride=0.999
    angularSpringConstant=200
    bAngularSpring=true
    bOverrideAngularDamping=true
    simulationSpace=RootRelative
    alphaBoolBlend {blendOption=Linear}
    boundBone {boneName=dyn_NameTag_02_jnt}
    componentAppliedLinearAccClamp=<6000 6000 6000>
    componentLinearAccScale=<0.5 0.5 0.5>
    componentLinearVelScale=<0.5 0.5 0.5>
    physicsBodyDefinitions [{
      boundBone {boneName=dyn_NameTag_02_jnt}
      boxExtents=<10 10 10>
      constraintSetup {angularLimitsMax=<2 0 1>, angularLimitsMin=<-2 -15 -1>, angularTarget=<1 0 0>}
      localJointOffset=<4 0 0>
     }]
   }
   n104 = Get Is Swimming Coat Alpha() -> (Is Swimming Coat Alpha:double)
   n105 = "Transform (Modify) Bone - Bone: dyn_Jacket_bk_c_MSTR"(ComponentPose=@n103, Alpha=@n104, Translation="0,0,0", Rotation="5,0,0", Scale="1,1,1") {
    rotationMode=BMM_Additive
    rotationSpace=BCS_BoneSpace
    translationMode=BMM_Additive
    translationSpace=BCS_BoneSpace
    alphaBoolBlend {blendOption=Linear}
    boneToModify {boneName=dyn_Jacket_bk_c_MSTR}
    rotation {pitch=5}
   }
   n106 = "Transform (Modify) Bone - Bone: dyn_Jacket_bk_l_MSTR"(ComponentPose=@n105, Alpha=@n104, Translation="0,0,0", Rotation="5,0,0", Scale="1,1,1") {
    rotationMode=BMM_Additive
    rotationSpace=BCS_BoneSpace
    translationMode=BMM_Additive
    translationSpace=BCS_BoneSpace
    alphaBoolBlend {blendOption=Linear}
    boneToModify {boneName=dyn_Jacket_bk_l_MSTR}
    rotation {pitch=5}
   }
   n107 = "Transform (Modify) Bone - Bone: dyn_Jacket_bk_r_MSTR"(ComponentPose=@n106, Alpha=@n104, Translation="0,0,0", Rotation="5,0,0", Scale="1,1,1") {
    rotationMode=BMM_Additive
    rotationSpace=BCS_BoneSpace
    translationMode=BMM_Additive
    translationSpace=BCS_BoneSpace
    alphaBoolBlend {blendOption=Linear}
    boneToModify {boneName=dyn_Jacket_bk_r_MSTR}
    rotation {pitch=5}
   }
   Output Pose(Result=Component To Local(ComponentPose=@n107)) {layerGroup=None}
  ]}
 EventGraph {nodes [
   Event BlueprintUpdateAnimation()
   > n1 = Sequence!(execute) -> (then_0, then_1, then_2:Exec, then_3:Exec, then_4:Exec, then_5:Exec, then_6:Exec, then_7:Exec, then_8:Exec)
   > n2 = Gravity Override State Function DEPRECATED(CharacterPartAnimInstance=Self-Reference(), GameTimeInSeconds=GetGameTimeInSeconds(), GravityOverrideParams=Get Gravity Override Params())
   > Set Gravity Override(Gravity Override=@n2)
   Set Is Crouching Alpha(execute=@n1.then_1, Is Crouching Alpha:double="To Float (Boolean)"(InBool=Get bIsCrouching())) -> (Output_Get:double)
   Set Is Playing Melee Anim Alpha(execute=@n1.then_4, Is Playing Melee Anim Alpha:double="To Float (Boolean)"(InBool=Get bIsPlayingMeleeAnim())) -> (Output_Get:double)
   n3 = Branch(execute=@n1.then_5, Condition=Get bIsTacticalSprinting())
   > Set Is Tactical Sprinting Coat Alpha(Is Tactical Sprinting Coat Alpha:double=0.75) -> (Output_Get:double)
   Set Is Tactical Sprinting Coat Alpha(execute=@n3.else, Is Tactical Sprinting Coat Alpha=1)
   Set Coat Alpha(
    execute=@n1.then_6
    Coat Alpha:double=float * float!(
     A=Get Is Tactical Sprinting Coat Alpha() -> (Is Tactical Sprinting Coat Alpha:double)
     B=Get Is Sprint Jumping Coat Alpha() -> (Is Sprint Jumping Coat Alpha:double)
     C:double=1
    ) -> (ReturnValue) {is=title K2Node_PromotableOperator}
   ) -> (Output_Get:double)
   n4 = NOT Boolean(A=Get bIsSkyDiving())
   n5 = Get bIsFalling()
   n6 = Branch(
    execute=@n1.then_7
    Condition=OR Boolean!(A=Get bIsJumping(), B=AND Boolean(A=@n5, B=@n4), C:bool=Get bIsTacticalSprinting()) -> (ReturnValue)
   )
   > Set Is Sprint Jumping Coat Alpha(Is Sprint Jumping Coat Alpha:double=0.75) -> (Output_Get:double)
   Set Is Sprint Jumping Coat Alpha(execute=@n6.else, Is Sprint Jumping Coat Alpha=1)
   n7 = Get bIsSurfaceSwimming()
   n8 = Get bIsSwimSprinting() -> (bIsSwimSprinting:bool)
   Set Is Swimming Coat Alpha(execute=@n1.then_8, Is Swimming Coat Alpha:double="To Float (Boolean)"(InBool=OR Boolean!(A=@n7, B=@n8) -> (ReturnValue))) -> (Output_Get:double)
   Branch(execute=@n1.then_3, Condition=Get Is Parachute Opened() -> (Is Parachute Opened:bool))
   > Set Is Parachute Opened Float(Is Parachute Opened Float:double=0) -> (Output_Get:double)
   > Delay(Duration=0.2)
   > Set Is Parachute Opened Float(Is Parachute Opened Float=1)
   > Delay(Duration=3)
   > Set Is Parachute Opened Float(Is Parachute Opened Float=0)
   n9 = Sequence!(execute=@n1.then_2) -> (then_0, then_1, then_2, then_3, then_4)
   > n10 = Cast To FortPlayerAnimInstance(Object=GetAnimInstance(self=GetOwnerSkeletalMeshComponent()))
   > Set isPistol(isPistol=false)
   > Set isRifle(isRifle:bool=false) -> (Output_Get:bool)
   > n11 = Switch on EFortWeaponCoreAnimation!(execute, Selection:EFortWeaponCoreAnimation Enum=Get WeaponCoreAnim(self=@n10)) -> (
    Melee:Exec
    Pistol:Exec
    Shotgun:Exec
    PaperBlueprint:Exec
    Rifle:Exec
    MeleeOneHand:Exec
    MachinePistol:Exec
    RocketLauncher:Exec
    GrenadeLauncher:Exec
    GoingCommando:Exec
    AssaultRifle:Exec
    TacticalShotgun:Exec
    SniperRifle:Exec
    TrapPlacement:Exec
    ShoulderLauncher:Exec
    AbilityDecoTool:Exec
    Crossbow:Exec
    C4:Exec
    RemoteControl:Exec
    DualWield:Exec
    AR_BullPup:Exec
    AR_ForwardGrip:Exec
    MedPackPaddles:Exec
    SMG_P90:Exec
    AR_DrumGun:Exec
    Consumable_Small:Exec
    Consumable_Large:Exec
    Balloon:Exec
    MountedTurret:Exec
    CreativeTool:Exec
    ExplosiveBow:Exec
    AshtonIndigo:Exec
    AshtonChicago:Exec
    MeleeDualWield:Exec
    Unarmed:Exec
   )
   Set isRifle(execute=[@n11.AssaultRifle, @n11.Rifle, @n11.RocketLauncher, @n11.Shotgun, @n11.SniperRifle, @n11.TacticalShotgun], isRifle=true)
   Set isPistol(execute=@n11.Pistol, isPistol=true)
   Set Is Rifle Alpha(execute=@n9.then_1, Is Rifle Alpha:double="To Float (Boolean)"(InBool=Get isRifle() -> (isRifle:bool))) -> (Output_Get:double)
   Set Is Pistol Alpha(execute=@n9.then_2, Is Pistol Alpha:double="To Float (Boolean)"(InBool=Get isPistol() -> (isPistol:bool))) -> (Output_Get:double)
   n12 = Get bIsTargeting()
   n13 = Get isRifle()
   Set Is Rifle Targeting Alpha(execute=@n9.then_3, Is Rifle Targeting Alpha:double="To Float (Boolean)"(InBool=AND Boolean(A=@n13, B=@n12))) -> (Output_Get:double)
   n14 = Get bIsTargeting()
   n15 = Get isPistol()
   Set Is Pistol Targeting Alpha(execute=@n9.then_4, Is Pistol Targeting Alpha:float="To Float (Boolean)"(InBool=AND Boolean(A=@n15, B=@n14))) -> (Output_Get:float)
   Event OnParachuteOpened()
   > Set Is Parachute Opened(Is Parachute Opened:bool=true) -> (Output_Get:bool)
   > Delay(Duration=0.2)
   > Set Is Parachute Opened(Is Parachute Opened=false)
  ]}
]
variables [
 Coat Alpha {type=number}
 Gravity Override {type=Vector}
 Gravity Override Params {type=GravityOverrideParamsStruct}
 Is Crouching Alpha {type=number}
 Is Parachute Opened {type=boolean}
 Is Parachute Opened Float {type=number}
 Is Pistol Alpha {type=number}
 Is Pistol Targeting Alpha {type=number}
 Is Playing Melee Anim Alpha {type=number}
 Is Rifle Alpha {type=number}
 Is Rifle Targeting Alpha {type=number}
 Is Sprint Jumping Coat Alpha {type=number}
 Is Swimming Coat Alpha {type=number}
 Is Tactical Sprinting Coat Alpha {type=number}
 isPistol {type=boolean}
 isRifle {type=boolean}
 skelMesh {type=SkeletalMeshComponent}
]
