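# Control Rig for the PrismParticle male character's face accessory mesh that poses the master bones of the three back hair chains.
# Rotation constraints blend each master between spine_05 and head, then spherical pose readers on head orientation weight corrective rotations, alongside two full-weight center offsets.
# Right upper-arm pose readers, multiplied by pistol or rifle alpha variables and smoothed by AlphaInterp, add further rotations for weapon poses.
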ControlRigBlueprint /BRCosmetics/Characters/Player/Male/Medium/Bodies/M_MED_PrismParticle/Meshes/Parts/M_MED_PrismParticle_FaceAcc_CtrlRig
graphs [
 RigVMFunctionLibrary {links [], nodes []}
 RigVMModel {nodes [
   RigUnit_BeginExecution = &RigUnit_BeginExecution()
   &RigUnit_RotationConstraintLocalSpaceOffset(
    AdvancedSettings=(RotationOrderForFilter=XZY)
    Child=(Type=Bone,Name="dyn_Hair_l_bk_MSTR")
    Filter=(bX=True,bY=True,bZ=True)
    Parents=((Item=(Type=Bone,Name="spine_05"),Weight=0.5),(Item=(Type=Bone,Name="head"),Weight=0.5))
    Weight=1
    ExecutePin=@RigUnit_BeginExecution.ExecutePin
   )
   > &RigUnit_RotationConstraintLocalSpaceOffset(
    AdvancedSettings=(RotationOrderForFilter=XZY)
    Child=(Type=Bone,Name="dyn_Hair_c_bk_MSTR")
    Filter=(bX=True,bY=True,bZ=True)
    Parents=((Item=(Type=Bone,Name="spine_05"),Weight=0.5),(Item=(Type=Bone,Name="head"),Weight=0.5))
    Weight=1
   )
   > &RigUnit_RotationConstraintLocalSpaceOffset(
    AdvancedSettings=(RotationOrderForFilter=XZY)
    Child=(Type=Bone,Name="dyn_Hair_r_bk_MSTR")
    Filter=(bX=True,bY=True,bZ=True)
    Parents=((Item=(Type=Bone,Name="spine_05"),Weight=0.5),(Item=(Type=Bone,Name="head"),Weight=0.5))
    Weight=1
    bMaintainOffset=true
   )
   > n1 = &RigUnit_SphericalPoseReader(
    ActiveRegionScaleFactors=(PositiveWidth=1,NegativeWidth=1,PositiveHeight=1,NegativeHeight=1)
    ActiveRegionSize=0.1
    Debug=(bDrawDebug=false,DebugScale=25,DebugSegments=20,DebugThickness=0.25)
    DriverAxis=(X=0,Y=-1,Z=0)
    DriverItem=(Type=Bone,Name="head")
    FalloffRegionScaleFactors=(PositiveWidth=2,NegativeWidth=2,PositiveHeight=2,NegativeHeight=2)
    FalloffSize=0.2
    FlipWidthScaling=false
    OptionalParentItem=(Type=Bone,Name="neck_01")
    RotationOffset=(X=-90,Y=0,Z=-55)
   )
   > &RigUnit_ModifyTransforms(
    ItemToModify=((Item=(Type=Bone,Name="dyn_Hair_r_bk_MSTR"),Transform=(Rotation=(X=0,Y=0,Z=0,W=1),Translation=(X=0,Y=0,Z=0),Scale3D=(X=1,Y=1,Z=1))))
    ItemToModify.0.Transform.Rotation=&RigVMFunction_MathQuaternionFromEuler(Euler=(X=0,Y=5,Z=20)).Result
    Weight=@n1.OutputParam
   )
   > &RigUnit_ModifyTransforms(
    ItemToModify=((Item=(Type=Bone,Name="dyn_Hair_l_bk_MSTR"),Transform=(Rotation=(X=0,Y=0,Z=0,W=1),Translation=(X=0,Y=0,Z=0),Scale3D=(X=1,Y=1,Z=1))))
    ItemToModify.0.Transform.Rotation=&RigVMFunction_MathQuaternionFromEuler(Euler=(X=0,Y=5,Z=-20)).Result
    Weight=@n1.OutputParam
   )
   > &RigUnit_ModifyTransforms(
    ItemToModify=((Item=(Type=Bone,Name="dyn_Hair_c_bk_MSTR"),Transform=(Rotation=(X=0,Y=0,Z=0,W=1),Translation=(X=0,Y=0,Z=0),Scale3D=(X=1,Y=1,Z=1))))
    ItemToModify.0.Transform.Rotation=&RigVMFunction_MathQuaternionFromEuler(Euler=(X=0,Y=0,Z=-20)).Result
    Weight=@n1.OutputParam
   )
   > n2 = &RigUnit_SphericalPoseReader(
    ActiveRegionScaleFactors=(PositiveWidth=1,NegativeWidth=1,PositiveHeight=3,NegativeHeight=0.5)
    ActiveRegionSize=0.1
    Debug=(bDrawDebug=false,DebugScale=25,DebugSegments=20,DebugThickness=0.25)
    DriverAxis=(X=0,Y=-1,Z=0)
    DriverItem=(Type=Bone,Name="head")
    FalloffRegionScaleFactors=(PositiveWidth=1,NegativeWidth=0.5,PositiveHeight=2,NegativeHeight=0.5)
    FalloffSize=0.3
    OptionalParentItem=(Type=Bone,Name="neck_02")
    RotationOffset=(X=-55,Y=0,Z=-10)
   )
   > &RigUnit_ModifyTransforms(
    ItemToModify=((Item=(Type=Bone,Name="dyn_Hair_r_bk_MSTR"),Transform=(Rotation=(X=0,Y=0,Z=0,W=1),Translation=(X=0,Y=0,Z=0),Scale3D=(X=1,Y=1,Z=1))))
    ItemToModify.0.Transform.Rotation=&RigVMFunction_MathQuaternionFromEuler(Euler=(X=0,Y=-15,Z=0)).Result
    Weight=@n2.OutputParam
   )
   > &RigUnit_ModifyTransforms(
    ItemToModify=((Item=(Type=Bone,Name="dyn_Hair_c_bk_MSTR"),Transform=(Rotation=(X=0,Y=0,Z=0,W=1),Translation=(X=0,Y=0,Z=0),Scale3D=(X=1,Y=1,Z=1))))
    Weight=1
    ItemToModify.0.Transform.Rotation=&RigVMFunction_MathQuaternionFromEuler(Euler=(X=0,Y=0,Z=-10)).Result
   )
   > n3 = &RigUnit_SphericalPoseReader(
    ActiveRegionScaleFactors=(PositiveWidth=1,NegativeWidth=1,PositiveHeight=0.5,NegativeHeight=3)
    ActiveRegionSize=0.1
    Debug=(bDrawDebug=false,DebugScale=25,DebugSegments=20,DebugThickness=0.25)
    DriverAxis=(X=0,Y=-1,Z=0)
    DriverItem=(Type=Bone,Name="head")
    FalloffRegionScaleFactors=(PositiveWidth=1,NegativeWidth=0.5,PositiveHeight=0.5,NegativeHeight=2)
    FalloffSize=0.3
    OptionalParentItem=(Type=Bone,Name="neck_02")
    RotationOffset=(X=235,Y=0,Z=-10)
   )
   > &RigUnit_ModifyTransforms(
    ItemToModify=((Item=(Type=Bone,Name="dyn_Hair_l_bk_MSTR"),Transform=(Rotation=(X=0,Y=0,Z=0,W=1),Translation=(X=0,Y=0,Z=0),Scale3D=(X=1,Y=1,Z=1))))
    ItemToModify.0.Transform.Rotation=&RigVMFunction_MathQuaternionFromEuler(Euler=(X=0,Y=-15,Z=0)).Result
    Weight=@n3.OutputParam
   )
   > &RigUnit_ModifyTransforms(
    ItemToModify=((Item=(Type=Bone,Name="dyn_Hair_c_bk_MSTR"),Transform=(Rotation=(X=0,Y=0,Z=0,W=1),Translation=(X=0,Y=0,Z=0),Scale3D=(X=1,Y=1,Z=1))))
    Weight=1
    ItemToModify.0.Transform.Rotation=&RigVMFunction_MathQuaternionFromEuler(Euler=(X=0,Y=0,Z=-10)).Result
   )
   > n4 = &RigUnit_SphericalPoseReader(
    ActiveRegionScaleFactors=(PositiveWidth=0.5,NegativeWidth=0.5,PositiveHeight=1,NegativeHeight=0.5)
    ActiveRegionSize=0.1
    Debug=(bDrawDebug=false,bDraw2D=false,bDrawLocalAxes=false,DebugScale=25,DebugSegments=20,DebugThickness=0.25)
    DriverAxis=(X=-1,Y=0,Z=0)
    DriverItem=(Type=Bone,Name="upperarm_r")
    FalloffRegionScaleFactors=(PositiveWidth=0.5,NegativeWidth=0.5,PositiveHeight=1,NegativeHeight=0.5)
    FalloffSize=0.45
    OptionalParentItem=(Type=Bone,Name="spine_05")
    RotationOffset=(X=-90,Y=40,Z=45)
   )
   Multiply = &RigVMFunction_MathDoubleMul(A=@n4.OutputParam, B=RigVMVariableNode() {variable=Is Pistol Alpha}.Value)
   n5 = &RigVMFunction_AlphaInterp(InterpSpeedDecreasing=4, InterpSpeedIncreasing=10, Scale=1, bClampResult=False, Value=@Multiply.Result)
   > &RigUnit_ModifyTransforms(
    ItemToModify=((Item=(Type=Bone,Name="dyn_Hair_r_bk_MSTR"),Transform=(Rotation=(X=0,Y=0,Z=0,W=1),Translation=(X=0,Y=0,Z=0),Scale3D=(X=1,Y=1,Z=1))))
    Weight=@n5.Result
    ItemToModify.0.Transform.Rotation=&RigVMFunction_MathQuaternionFromEuler(Euler=(X=0,Y=-20,Z=20)).Result
   )
   > &RigUnit_ModifyTransforms(
    ItemToModify=((Item=(Type=Bone,Name="dyn_Hair_l_bk_MSTR"),Transform=(Rotation=(X=0,Y=0,Z=0,W=1),Translation=(X=0,Y=0,Z=0),Scale3D=(X=1,Y=1,Z=1))))
    Weight=@n5.Result
    ItemToModify.0.Transform.Rotation=&RigVMFunction_MathQuaternionFromEuler(Euler=(X=0,Y=-20,Z=20)).Result
   )
   > &RigUnit_ModifyTransforms(
    ItemToModify=((Item=(Type=Bone,Name="dyn_Hair_c_bk_MSTR"),Transform=(Rotation=(X=0,Y=0,Z=0,W=1),Translation=(X=0,Y=0,Z=0),Scale3D=(X=1,Y=1,Z=1))))
    Weight=@n5.Result
    ItemToModify.0.Transform.Rotation=&RigVMFunction_MathQuaternionFromEuler(Euler=(X=0,Y=0,Z=20)).Result
   )
   > n6 = &RigUnit_SphericalPoseReader(
    ActiveRegionScaleFactors=(PositiveWidth=0.35,NegativeWidth=0.35,PositiveHeight=0.35,NegativeHeight=0.35)
    ActiveRegionSize=0.1
    Debug=(bDrawDebug=false,bDraw2D=false,bDrawLocalAxes=false,DebugScale=25,DebugSegments=20,DebugThickness=0.25)
    DriverAxis=(X=-1,Y=0,Z=0)
    DriverItem=(Type=Bone,Name="upperarm_r")
    FalloffRegionScaleFactors=(PositiveWidth=0.5,NegativeWidth=0.5,PositiveHeight=0.5,NegativeHeight=0.5)
    FalloffSize=0.2
    OptionalParentItem=(Type=Bone,Name="spine_05")
    RotationOffset=(X=-45,Y=0,Z=-30)
   )
   n7 = &RigVMFunction_AlphaInterp(
    InterpSpeedDecreasing=4
    InterpSpeedIncreasing=10
    Scale=1
    bClampResult=False
    Value=&RigVMFunction_MathDoubleMul(A=@n6.OutputParam, B=RigVMVariableNode() {variable=Is Rifle Alpha}.Value).Result
   )
   > &RigUnit_ModifyTransforms(
    ItemToModify=((Item=(Type=Bone,Name="dyn_Hair_r_bk_MSTR"),Transform=(Rotation=(X=0,Y=0,Z=0,W=1),Translation=(X=0,Y=0,Z=0),Scale3D=(X=1,Y=1,Z=1))))
    Weight=@n7.Result
    ItemToModify.0.Transform.Rotation=&RigVMFunction_MathQuaternionFromEuler(Euler=(X=0,Y=-20,Z=20)).Result
   )
   > &RigUnit_ModifyTransforms(
    ItemToModify=((Item=(Type=Bone,Name="dyn_Hair_l_bk_MSTR"),Transform=(Rotation=(X=0,Y=0,Z=0,W=1),Translation=(X=0,Y=0,Z=0),Scale3D=(X=1,Y=1,Z=1))))
    Weight=@n7.Result
    ItemToModify.0.Transform.Rotation=&RigVMFunction_MathQuaternionFromEuler(Euler=(X=0,Y=-20,Z=20)).Result
   )
   > &RigUnit_ModifyTransforms(
    ItemToModify=((Item=(Type=Bone,Name="dyn_Hair_c_bk_MSTR"),Transform=(Rotation=(X=0,Y=0,Z=0,W=1),Translation=(X=0,Y=0,Z=0),Scale3D=(X=1,Y=1,Z=1))))
    Weight=@n7.Result
    ItemToModify.0.Transform.Rotation=&RigVMFunction_MathQuaternionFromEuler(Euler=(X=0,Y=0,Z=20)).Result
   )
  ]}
]
hierarchy [
 root {loc=<0 0 0>, rot=<0 0 0 1>}
 pelvis {parent=root, loc=<0 -0.7071 93.1191>, rot=<0.0127 0.707 -0.0127 -0.707>}
 spine_01 {parent=pelvis, loc=<6.7494 -0.6388 0>, rot=<0 0 0.034 -0.9994>}
 spine_02 {parent=spine_01, loc=<7.2518 -0.035 0>, rot=<0 0 0.0238 -0.9997>}
 spine_03 {parent=spine_02, loc=<6.0104 0 0>, rot=<0 0 0.0149 -0.9999>}
 spine_04 {parent=spine_03, loc=<8.1024 0 0>, rot=<0 0 -0.0132 -0.9999>}
 spine_05 {parent=spine_04, loc=<10.0146 0 0>, rot=<0 0 -0.0015 -1>}
 neck_01 {parent=spine_05, loc=<11.3138 0.1529 0>, rot=<0 0 -0.2003 -0.9797>}
 neck_02 {parent=neck_01, loc=<4.5627 0 0>, rot=<0 0 0.0504 -0.9987>}
 head {parent=neck_02, loc=<5.3914 0 0>, rot=<0 0 0.115 -0.9934>}
 dyn_Hair_c_bk_MSTR {parent=head, loc=<6.0133 -7.7355 0>, rot=<0.1357 -0.9908 0 0>}
 dyn_Hair_c_bk_01_jnt {parent=dyn_Hair_c_bk_MSTR, loc=<0 0 0>, rot=<0 0 0 1>}
 dyn_Hair_c_bk_02_jnt {parent=dyn_Hair_c_bk_01_jnt, loc=<3.959 0 0>, rot=<0 0 0.022 -0.9998>}
 dyn_Hair_c_bk_03_jnt {parent=dyn_Hair_c_bk_02_jnt, loc=<3.959 0 0>, rot=<0 0 0.022 -0.9998>}
 dyn_Hair_c_bk_04_jnt {parent=dyn_Hair_c_bk_03_jnt, loc=<3.959 0 0>, rot=<0 0 -0.0471 -0.9989>}
 dyn_Hair_c_bk_05_jnt {parent=dyn_Hair_c_bk_04_jnt, loc=<3.959 0 0>, rot=<0 0 -0.0597 -0.9982>}
 dyn_Hair_r_bk_MSTR {parent=head, loc=<7.8755 4.6681 7.3004>, rot=<0.0973 0.034 0.87 -0.4821>}
 dyn_Hair_r_bk_01_jnt {parent=dyn_Hair_r_bk_MSTR, loc=<0 0 0>, rot=<0 0 0 1>}
 dyn_Hair_r_bk_02_jnt {parent=dyn_Hair_r_bk_01_jnt, loc=<4.6569 0 0>, rot=<0 0 0.1414 -0.99>}
 dyn_Hair_r_bk_03_jnt {parent=dyn_Hair_r_bk_02_jnt, loc=<4.6569 0 0>, rot=<0 0 0.1248 -0.9922>}
 dyn_Hair_r_bk_04_jnt {parent=dyn_Hair_r_bk_03_jnt, loc=<4.6569 0 0>, rot=<0 0 -0.0783 -0.9969>}
 dyn_Hair_r_bk_05_jnt {parent=dyn_Hair_r_bk_04_jnt, loc=<4.6569 0 0>, rot=<0 0 -0.0783 -0.9969>}
 dyn_Hair_l_bk_MSTR {parent=head, loc=<7.8757 4.6681 -7.3004>, rot=<0.4821 -0.87 -0.034 -0.0973>}
 dyn_Hair_l_bk_01_jnt {parent=dyn_Hair_l_bk_MSTR, loc=<0 0 0>, rot=<0 0 0 1>}
 dyn_Hair_l_bk_02_jnt {parent=dyn_Hair_l_bk_01_jnt, loc=<4.6567 0 0>, rot=<0 0 -0.1414 -0.99>}
 dyn_Hair_l_bk_03_jnt {parent=dyn_Hair_l_bk_02_jnt, loc=<4.6567 0 0>, rot=<0 0 -0.1248 -0.9922>}
 dyn_Hair_l_bk_04_jnt {parent=dyn_Hair_l_bk_03_jnt, loc=<4.6567 0 0>, rot=<0 0 0.0783 -0.9969>}
 dyn_Hair_l_bk_05_jnt {parent=dyn_Hair_l_bk_04_jnt, loc=<4.6567 0 0>, rot=<0 0 0.0783 -0.9969>}
 dyn_Hair_tp_MSTR {parent=head, loc=<17.6128 6.5901 -2.1788>, rot=<0.597 -0.501 0.1886 -0.5975>}
 dyn_Hair_tp_01_jnt {parent=dyn_Hair_tp_MSTR, loc=<0 0 0>, rot=<0 0 0 1>}
 dyn_Hair_tp_02_jnt {parent=dyn_Hair_tp_01_jnt, loc=<2.5048 0 0>, rot=<-0.0021 -0.1898 -0.0108 -0.9818>}
 dyn_Hair_tp_03_jnt {parent=dyn_Hair_tp_02_jnt, loc=<2.6064 0 0>, rot=<-0.0033 -0.2092 -0.0154 -0.9777>}
 dyn_Hair_tp_04_jnt {parent=dyn_Hair_tp_03_jnt, loc=<2.6064 0 0>, rot=<0.0143 -0.172 0.0814 -0.9816>}
 dyn_Hair_tp_05_jnt {parent=dyn_Hair_tp_04_jnt, loc=<2.0935 0 0>, rot=<0.0104 0.0491 0.2062 -0.9772>}
 dyn_Hair_tp_06_jnt {parent=dyn_Hair_tp_05_jnt, loc=<3.1514 0 0>, rot=<0.0058 0.2389 -0.1397 -0.9609>}
 dyn_Hair_ft_MSTR {parent=head, loc=<15.3468 8.0188 2.6557>, rot=<-0.7048 -0.116 -0.381 -0.587>}
 dyn_Hair_ft_01_jnt {parent=dyn_Hair_ft_MSTR, loc=<0 0 0>, rot=<0 0 0 1>}
 dyn_Hair_ft_02_jnt {parent=dyn_Hair_ft_01_jnt, loc=<2.4121 0 0>, rot=<0.0391 -0.5547 -0.0289 -0.8306>}
 dyn_Hair_ft_03_jnt {parent=dyn_Hair_ft_02_jnt, loc=<3.4414 0 0>, rot=<0.0573 -0.4751 0.0063 -0.878>}
 dyn_Hair_ft_04_jnt {parent=dyn_Hair_ft_03_jnt, loc=<3.4414 0 0>, rot=<0.0567 -0.1892 0.1182 -0.9731>}
 dyn_Hair_ft_05_jnt {parent=dyn_Hair_ft_04_jnt, loc=<2.931 0 0>, rot=<0.091 0.2707 -0.2217 -0.9324>}
 FACIAL_C_FacialRoot {parent=head, loc=<2.9906 3.8922 0>, rot=<0.4853 -0.5142 -0.4854 -0.5142>}
 FACIAL_C_Forehead {parent=FACIAL_C_FacialRoot, loc=<-0.0002 -3.1198 8.3218>, rot=<-0.0267 0 0.0001 -0.9996>}
 FACIAL_L_ForeheadIn {parent=FACIAL_C_FacialRoot, loc=<1.3636 -3.0576 8.2573>, rot=<-0.0441 -0.0695 0.0056 -0.9966>}
 FACIAL_R_ForeheadIn {parent=FACIAL_C_FacialRoot, loc=<-1.3755 -3.0655 8.2605>, rot=<-0.0439 0.0695 -0.0053 -0.9966>}
 FACIAL_L_ForeheadMid {parent=FACIAL_C_FacialRoot, loc=<2.6906 -3.245 7.8324>, rot=<-0.0548 -0.176 -0.0176 -0.9827>}
 FACIAL_R_ForeheadMid {parent=FACIAL_C_FacialRoot, loc=<-2.6995 -3.2461 7.8323>, rot=<-0.0546 0.176 0.0178 -0.9827>}
 FACIAL_L_ForeheadOut {parent=FACIAL_C_FacialRoot, loc=<4.2598 -3.0918 6.6489>, rot=<-0.1076 -0.2567 -0.0753 -0.9575>}
 FACIAL_R_ForeheadOut {parent=FACIAL_C_FacialRoot, loc=<-4.2724 -3.0928 6.6462>, rot=<-0.1074 0.2567 0.0755 -0.9575>}
 FACIAL_L_EyesackUpper {parent=FACIAL_C_FacialRoot, loc=<2.9151 -2.6137 7.8288>, rot=<0.0636 -0.0852 -0.0632 -0.9923>}
 FACIAL_R_EyesackUpper {parent=FACIAL_C_FacialRoot, loc=<-2.9229 -2.6101 7.834>, rot=<0.0637 0.0852 0.0633 -0.9923>}
 FACIAL_L_EyelidUpperFurrow {parent=FACIAL_C_FacialRoot, loc=<2.8204 -2.2913 7.3822>, rot=<-0.0169 -0.1519 -0.0776 -0.9852>}
 FACIAL_R_EyelidUpperA {parent=FACIAL_C_FacialRoot, loc=<-2.4333 -1.8482 6.0273>, rot=<-0.1367 0 0 -0.9906>}
 FACIAL_L_Eye {parent=FACIAL_C_FacialRoot, loc=<2.4525 -1.8856 6.0709>, rot=<0.0113 0 0.0001 -0.9999>}
 FACIAL_R_Eye {parent=FACIAL_C_FacialRoot, loc=<-2.4657 -1.8859 6.0605>, rot=<0.0112 0 0 -0.9999>}
 FACIAL_R_EyelidLowerA {parent=FACIAL_C_FacialRoot, loc=<-2.4333 -1.8482 6.0273>, rot=<0.2947 0 0 -0.9556>}
 FACIAL_L_EyeCornerOuter {parent=FACIAL_C_FacialRoot, loc=<3.8811 -1.6761 6.0811>, rot=<0.0104 -0.3827 0.0044 -0.9238>}
 FACIAL_R_EyeCornerOuter {parent=FACIAL_C_FacialRoot, loc=<-3.8881 -1.6751 6.0722>, rot=<0.0104 0.3826 -0.0043 -0.9238>}
 FACIAL_L_NasolabialBulge {parent=FACIAL_C_FacialRoot, loc=<3.2674 3.5576 5.005>, rot=<0.0777 -0.3768 0.032 -0.9225>}
 FACIAL_R_NasolabialBulge {parent=FACIAL_C_FacialRoot, loc=<-3.2757 3.5537 5.0006>, rot=<0.0778 0.3768 -0.0318 -0.9225>}
 FACIAL_L_CheekLower {parent=FACIAL_C_FacialRoot, loc=<4.0188 2.7865 4.328>, rot=<0.0539 -0.4256 0.067 -0.9008>}
 FACIAL_R_CheekLower {parent=FACIAL_C_FacialRoot, loc=<-4.0283 2.7809 4.3236>, rot=<0.054 0.4256 -0.0668 -0.9008>}
 FACIAL_C_MouthUpper {parent=FACIAL_C_FacialRoot, loc=<0.0001 3.7204 5.7772>, rot=<0.0215 0 0.0001 -0.9998>}
 FACIAL_L_LipCorner {parent=FACIAL_C_MouthUpper, loc=<2.5013 0.3927 0.602>, rot=<0 -0.2079 0 -0.9782>}
 FACIAL_R_LipCorner {parent=FACIAL_C_MouthUpper, loc=<-2.5029 0.3916 0.5995>, rot=<0 0.2079 0 -0.9782>}
 FACIAL_L_Masseter {parent=FACIAL_C_FacialRoot, loc=<5.071 1.2619 3.3039>, rot=<0.0905 0.113 0.0056 -0.9894>}
 FACIAL_R_Masseter {parent=FACIAL_C_FacialRoot, loc=<-5.0842 1.258 3.305>, rot=<0.0904 -0.113 -0.0054 -0.9895>}
 FACIAL_C_UnderChin {parent=FACIAL_C_FacialRoot, loc=<-0.0004 7.2808 1.9838>, rot=<0.5088 -0.0002 0 -0.8609>}
 FACIAL_L_UnderChin {parent=FACIAL_C_FacialRoot, loc=<2.6611 5.7193 1.9282>, rot=<0.42 -0.3535 0.0897 -0.831>}
 FACIAL_R_UnderChin {parent=FACIAL_C_FacialRoot, loc=<-2.6594 5.7035 1.9142>, rot=<0.4199 0.3527 -0.0898 -0.8314>}
 FACIAL_C_LowerLipRotation {parent=FACIAL_C_FacialRoot, loc=<0.0001 1.7805 1.4076>, rot=<0.3621 0 0.0001 -0.9321>}
 FACIAL_C_MouthLower {parent=FACIAL_C_LowerLipRotation, loc=<-0.0004 -1.5188 4.5332>, rot=<-0.342 0 0 -0.9397>}
 FACIAL_L_LipLowerOuter {parent=FACIAL_C_MouthLower, loc=<1.9222 0.637 1.1405>, rot=<0.0913 -0.2322 0.0318 -0.9679>}
 FACIAL_R_LipLowerOuter {parent=FACIAL_C_MouthLower, loc=<-1.9207 0.6379 1.1515>, rot=<0.0913 0.2322 -0.0318 -0.9679>}
 FACIAL_C_TeethLower {parent=FACIAL_C_LowerLipRotation, loc=<-0.0004 -1.913 5.6339>, rot=<-0.342 0 0 -0.9397>}
 FACIAL_C_Tongue1 {parent=FACIAL_C_TeethLower, loc=<0 0.116 -5.0732>, rot=<-0.2588 0 0 -0.9659>}
 FACIAL_C_Tongue2 {parent=FACIAL_C_Tongue1, loc=<0.0001 0.395 1.4872>, rot=<0.2588 0 0 -0.9659>}
 FACIAL_C_Tongue3 {parent=FACIAL_C_Tongue2, loc=<0 0.187 1.3384>, rot=<0.1305 0 0 -0.9914>}
 FACIAL_C_TongueLower3 {parent=FACIAL_C_Tongue3, loc=<0.0001 0.7038 0.0557>, rot=<0 0 0 1>}
 FACIAL_C_TongueUpper2 {parent=FACIAL_C_Tongue2, loc=<0.0001 -0.4669 0.111>, rot=<0 0 0 1>}
 FACIAL_L_TongueSide2 {parent=FACIAL_C_Tongue2, loc=<1.5871 0.2399 0.3633>, rot=<0 0 0 1>}
 FACIAL_R_TongueSide2 {parent=FACIAL_C_Tongue2, loc=<-1.5882 0.237 0.3619>, rot=<0 0 0 1>}
 FACIAL_C_Jaw {parent=FACIAL_C_FacialRoot, loc=<0.0001 1.7805 1.4076>, rot=<0.3621 0 0.0001 -0.9321>}
 FACIAL_C_Jawline {parent=FACIAL_C_Jaw, loc=<-0.0013 2.9612 6.2442>, rot=<0.2478 0 -0.0001 -0.9688>}
 FACIAL_L_ChinSide {parent=FACIAL_C_Jaw, loc=<2.3809 0.5908 5.9927>, rot=<-0.2249 -0.374 0.0481 -0.8985>}
 FACIAL_R_ChinSide {parent=FACIAL_C_Jaw, loc=<-2.3806 0.5877 5.9885>, rot=<-0.225 0.3739 -0.0481 -0.8985>}
 FACIAL_C_Chin {parent=FACIAL_C_Jaw, loc=<-0.0011 1.5206 6.7793>, rot=<-0.2716 0 0 -0.9624>}
 FACIAL_L_Ear {parent=FACIAL_C_FacialRoot, loc=<5.5459 0.0094 0.0726>, rot=<-0.0243 -0.0007 -0.0262 -0.9994>}
 FACIAL_R_Ear {parent=FACIAL_C_FacialRoot, loc=<-5.5554 -0.0208 0.0648>, rot=<-0.0251 0.0007 0.0262 -0.9993>}
 FACIAL_L_Jawline {parent=FACIAL_C_FacialRoot, loc=<3.6396 4.9288 3.5657>, rot=<0.2369 0.1557 -0.0592 -0.9572>}
 FACIAL_R_Jawline {parent=FACIAL_C_FacialRoot, loc=<-3.6364 4.9239 3.558>, rot=<0.2367 -0.1567 0.0592 -0.957>}
 FACIAL_L_EyelidLowerB {parent=FACIAL_C_FacialRoot, loc=<2.4525 -1.8856 6.0709>, rot=<0.328 0 0.0001 -0.9447>}
 FACIAL_R_EyelidLowerB {parent=FACIAL_C_FacialRoot, loc=<-2.4333 -1.8482 6.0273>, rot=<0.3361 0 0 -0.9418>}
 FACIAL_R_EyelidUpperB {parent=FACIAL_C_FacialRoot, loc=<-2.4333 -1.8482 6.0273>, rot=<-0.214 0 0 -0.9768>}
 clavicle_l {parent=spine_05, loc=<4.3575 -0.1466 -2.3223>, rot=<0.0465 -0.7289 -0.009 -0.6829>}
 upperarm_l {parent=clavicle_l, loc=<15.5111 0 0>, rot=<0.0226 -0.3234 -0.0088 -0.9459>}
 upperarm_twist_01_l {parent=upperarm_l, loc=<0.0041 0 0>, rot=<0 0 0 1>}
 clavicle_r {parent=spine_05, loc=<4.3575 -0.1466 2.3223>, rot=<0.7289 0.0465 -0.6829 0.009>}
 upperarm_r {parent=clavicle_r, loc=<-15.5111 0 0>, rot=<0.0226 -0.3234 -0.0088 -0.9459>}
 upperarm_twist_01_r {parent=upperarm_r, loc=<-0.0041 0 0>, rot=<0 0 0 -1>}
]
variables [
 Is Pistol Alpha {type=double, default=0}
 Is Rifle Alpha {type=double, default=0}
]
