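# Physics asset for the PrismParticle male FaceAcc mesh, setting up physics for five hair chains: center, left and right back, front, and top.
# Each chain runs from a MSTR bone on the head through capsule-bodied joints with narrow swing and very tight twist limits.
# Side back chains use longer angled capsules and a tighter second swing limit.
# Head, neck, upper spine and clavicle capsules form the body volume, joined by free-rotation constraints like the MSTR links.
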PhysicsAsset /BRCosmetics/Characters/Player/Male/Medium/Bodies/M_MED_PrismParticle/Meshes/Parts/M_MED_PrismParticle_FaceAcc_Physics
bodies {
 clavicle_l [{shapeType=Capsule, length=12, radius=8.5, shapeName=clavicle_l_capsule, center=<8 0 0.5>, rotation=<rot 90 0 0>}]
 clavicle_r [{shapeType=Capsule, length=12, radius=8.5, shapeName=clavicle_r_capsule, center=<-8 0 -0.5>, rotation=<rot 90 0 0>}]
 dyn_Hair_c_bk_01_jnt [{shapeType=Capsule, length=5, radius=3, shapeName=dyn_Hair_c_bk_01_jnt_capsule, center=<3 0 0>}]
 dyn_Hair_c_bk_02_jnt [{shapeType=Capsule, length=5, radius=3, shapeName=dyn_Hair_c_bk_02_jnt_capsule, center=<3 0 0>}]
 dyn_Hair_c_bk_03_jnt [{shapeType=Capsule, length=5, radius=3, shapeName=dyn_Hair_c_bk_03_jnt_capsule, center=<3 0 0>}]
 dyn_Hair_c_bk_04_jnt [{shapeType=Capsule, length=5, radius=3, shapeName=dyn_Hair_c_bk_04_jnt_capsule, center=<3 -1 0>}]
 dyn_Hair_c_bk_05_jnt [{shapeType=Capsule, length=5, radius=3, shapeName=dyn_Hair_c_bk_05_jnt_capsule, center=<3 -2 0>}]
 dyn_Hair_c_bk_MSTR [{shapeType=Capsule, length=1, radius=1, shapeName=dyn_Hair_c_bk_MSTR_capsule}]
 dyn_Hair_ft_01_jnt [{shapeType=Capsule, length=2, radius=1, shapeName=dyn_Hair_ft_01_jnt_capsule, center=<1 0 0>, rotation=<rot 0 0 90>}]
 dyn_Hair_ft_02_jnt [{shapeType=Capsule, length=2, radius=1, shapeName=dyn_Hair_ft_02_jnt_capsule, center=<1 0 0>, rotation=<rot 0 0 90>}]
 dyn_Hair_ft_03_jnt [{shapeType=Capsule, length=2, radius=1, shapeName=dyn_Hair_ft_03_jnt_capsule, center=<1 0 0>, rotation=<rot 0 0 90>}]
 dyn_Hair_ft_04_jnt [{shapeType=Capsule, length=2, radius=1, shapeName=dyn_Hair_ft_04_jnt_capsule, center=<1 0 0>, rotation=<rot 0 0 90>}]
 dyn_Hair_ft_05_jnt [{shapeType=Capsule, length=2, radius=1, shapeName=dyn_Hair_ft_05_jnt_capsule, center=<1 0 0>, rotation=<rot 0 0 90>}]
 dyn_Hair_ft_MSTR [{shapeType=Capsule, length=1, radius=1, shapeName=dyn_Hair_ft_MSTR_capsule}]
 dyn_Hair_l_bk_01_jnt [{shapeType=Capsule, length=10, radius=3, shapeName=dyn_Hair_l_bk_01_jnt_capsule, center=<3 -2 -3>, rotation=<rot 22.521 20.3607 44.0953>}]
 dyn_Hair_l_bk_02_jnt [{shapeType=Capsule, length=10, radius=3, shapeName=dyn_Hair_l_bk_02_jnt_capsule, center=<3 -2 -3>, rotation=<rot 22.521 20.3607 44.0953>}]
 dyn_Hair_l_bk_03_jnt [{shapeType=Capsule, length=10, radius=3, shapeName=dyn_Hair_l_bk_03_jnt_capsule, center=<3 -2 -3>, rotation=<rot 22.521 20.3607 44.0953>}]
 dyn_Hair_l_bk_04_jnt [{shapeType=Capsule, length=10, radius=3, shapeName=dyn_Hair_l_bk_04_jnt_capsule, center=<3 -2 -3>, rotation=<rot 22.521 20.3607 44.0953>}]
 dyn_Hair_l_bk_05_jnt [{shapeType=Capsule, length=10, radius=3, shapeName=dyn_Hair_l_bk_05_jnt_capsule, center=<3 -2 -3>, rotation=<rot 22.521 20.3607 44.0953>}]
 dyn_Hair_l_bk_MSTR [{shapeType=Capsule, length=1, radius=1, shapeName=dyn_Hair_l_bk_MSTR_capsule}]
 dyn_Hair_r_bk_01_jnt [{shapeType=Capsule, length=10, radius=3, shapeName=dyn_Hair_r_bk_01_jnt_capsule, center=<3 2 -3>, rotation=<rot 22.521 -20.3608 -44.0953>}]
 dyn_Hair_r_bk_02_jnt [{shapeType=Capsule, length=10, radius=3, shapeName=dyn_Hair_r_bk_02_jnt_capsule, center=<3 2 -3>, rotation=<rot 22.521 -20.3608 -44.0953>}]
 dyn_Hair_r_bk_03_jnt [{shapeType=Capsule, length=10, radius=3, shapeName=dyn_Hair_r_bk_03_jnt_capsule, center=<3 2 -3>, rotation=<rot 22.521 -20.3608 -44.0953>}]
 dyn_Hair_r_bk_04_jnt [{shapeType=Capsule, length=10, radius=3, shapeName=dyn_Hair_r_bk_04_jnt_capsule, center=<3 2 -3>, rotation=<rot 22.521 -20.3608 -44.0953>}]
 dyn_Hair_r_bk_05_jnt [{shapeType=Capsule, length=10, radius=3, shapeName=dyn_Hair_r_bk_05_jnt_capsule, center=<3 2 -3>, rotation=<rot 22.521 -20.3608 -44.0953>}]
 dyn_Hair_r_bk_MSTR [{shapeType=Capsule, length=1, radius=1, shapeName=dyn_Hair_r_bk_MSTR_capsule}]
 dyn_Hair_tp_01_jnt [{shapeType=Capsule, length=3, radius=1, shapeName=dyn_Hair_tp_01_jnt_capsule, center=<1 0 0>, rotation=<rot 0 0 90>}]
 dyn_Hair_tp_02_jnt [{shapeType=Capsule, length=3, radius=1, shapeName=dyn_Hair_tp_02_jnt_capsule, center=<1 0 0>, rotation=<rot 0 0 90>}]
 dyn_Hair_tp_03_jnt [{shapeType=Capsule, length=3, radius=1, shapeName=dyn_Hair_tp_03_jnt_capsule, center=<1 0 0>, rotation=<rot 0 0 90>}]
 dyn_Hair_tp_04_jnt [{shapeType=Capsule, length=3, radius=1, shapeName=dyn_Hair_tp_04_jnt_capsule, center=<1 0 0>, rotation=<rot 0 0 90>}]
 dyn_Hair_tp_05_jnt [{shapeType=Capsule, length=3, radius=1, shapeName=dyn_Hair_tp_05_jnt_capsule, center=<1 0 0>, rotation=<rot 0 0 90>}]
 dyn_Hair_tp_06_jnt [{shapeType=Capsule, length=3, radius=1, shapeName=dyn_Hair_tp_06_jnt_capsule, center=<1 0 0>, rotation=<rot 0 0 90>}]
 dyn_Hair_tp_MSTR [{shapeType=Capsule, length=1, radius=1, shapeName=dyn_Hair_tp_MSTR_capsule}]
 head [{shapeType=Capsule, length=1, radius=8, shapeName=head_capsule, center=<2.8953 4.4579 0>, rotation=<rot 90 0 0>}]
 neck_01 [{shapeType=Capsule, length=5, radius=7, shapeName=neck_01_capsule, center=<0 1 0>, rotation=<rot 90 0 0>}]
 neck_02 [{shapeType=Capsule, length=5, radius=6, shapeName=neck_02_capsule, center=<0 1 0>, rotation=<rot 90 0 0>}]
 spine_05 [{shapeType=Capsule, length=20, radius=12, shapeName=spine_05_capsule, center=<2 2 0>}]
}
constraints [
 {
  bone1Name=dyn_Hair_c_bk_MSTR
  bone2Name=head
  swing1LimitDegrees=45
  swing1Motion=Free
  swing2LimitDegrees=45
  swing2Motion=Free
  twistLimitDegrees=45
  twistMotion=Free
 }
 {bone1Name=dyn_Hair_c_bk_01_jnt, bone2Name=dyn_Hair_c_bk_MSTR, swing1LimitDegrees=12, swing2LimitDegrees=12, twistLimitDegrees=2}
 {bone1Name=dyn_Hair_c_bk_02_jnt, bone2Name=dyn_Hair_c_bk_01_jnt, swing1LimitDegrees=12, swing2LimitDegrees=12, twistLimitDegrees=2}
 {bone1Name=dyn_Hair_c_bk_03_jnt, bone2Name=dyn_Hair_c_bk_02_jnt, swing1LimitDegrees=12, swing2LimitDegrees=12, twistLimitDegrees=2}
 {bone1Name=dyn_Hair_c_bk_04_jnt, bone2Name=dyn_Hair_c_bk_03_jnt, swing1LimitDegrees=12, swing2LimitDegrees=12, twistLimitDegrees=2}
 {bone1Name=dyn_Hair_c_bk_05_jnt, bone2Name=dyn_Hair_c_bk_04_jnt, swing1LimitDegrees=12, swing2LimitDegrees=12, twistLimitDegrees=2}
 {
  bone1Name=dyn_Hair_ft_MSTR
  bone2Name=head
  swing1LimitDegrees=45
  swing1Motion=Free
  swing2LimitDegrees=45
  swing2Motion=Free
  twistLimitDegrees=45
  twistMotion=Free
 }
 {bone1Name=dyn_Hair_ft_01_jnt, bone2Name=dyn_Hair_ft_MSTR, swing1LimitDegrees=12, swing2LimitDegrees=12, twistLimitDegrees=2}
 {bone1Name=dyn_Hair_ft_02_jnt, bone2Name=dyn_Hair_ft_01_jnt, swing1LimitDegrees=12, swing2LimitDegrees=12, twistLimitDegrees=2}
 {bone1Name=dyn_Hair_ft_03_jnt, bone2Name=dyn_Hair_ft_02_jnt, swing1LimitDegrees=12, swing2LimitDegrees=12, twistLimitDegrees=2}
 {bone1Name=dyn_Hair_ft_04_jnt, bone2Name=dyn_Hair_ft_03_jnt, swing1LimitDegrees=12, swing2LimitDegrees=12, twistLimitDegrees=2}
 {bone1Name=dyn_Hair_ft_05_jnt, bone2Name=dyn_Hair_ft_04_jnt, swing1LimitDegrees=12, swing2LimitDegrees=12, twistLimitDegrees=2}
 {
  bone1Name=dyn_Hair_l_bk_MSTR
  bone2Name=head
  swing1LimitDegrees=45
  swing1Motion=Free
  swing2LimitDegrees=45
  swing2Motion=Free
  twistLimitDegrees=45
  twistMotion=Free
 }
 {bone1Name=dyn_Hair_l_bk_01_jnt, bone2Name=dyn_Hair_l_bk_MSTR, swing1LimitDegrees=12, swing2LimitDegrees=6, twistLimitDegrees=2}
 {bone1Name=dyn_Hair_l_bk_02_jnt, bone2Name=dyn_Hair_l_bk_01_jnt, swing1LimitDegrees=12, swing2LimitDegrees=6, twistLimitDegrees=2}
 {bone1Name=dyn_Hair_l_bk_03_jnt, bone2Name=dyn_Hair_l_bk_02_jnt, swing1LimitDegrees=12, swing2LimitDegrees=6, twistLimitDegrees=2}
 {bone1Name=dyn_Hair_l_bk_04_jnt, bone2Name=dyn_Hair_l_bk_03_jnt, swing1LimitDegrees=12, swing2LimitDegrees=6, twistLimitDegrees=2}
 {bone1Name=dyn_Hair_l_bk_05_jnt, bone2Name=dyn_Hair_l_bk_04_jnt, swing1LimitDegrees=12, swing2LimitDegrees=6, twistLimitDegrees=2}
 {
  bone1Name=dyn_Hair_r_bk_MSTR
  bone2Name=head
  swing1LimitDegrees=45
  swing1Motion=Free
  swing2LimitDegrees=45
  swing2Motion=Free
  twistLimitDegrees=45
  twistMotion=Free
 }
 {bone1Name=dyn_Hair_r_bk_01_jnt, bone2Name=dyn_Hair_r_bk_MSTR, swing1LimitDegrees=12, swing2LimitDegrees=6, twistLimitDegrees=2}
 {bone1Name=dyn_Hair_r_bk_02_jnt, bone2Name=dyn_Hair_r_bk_01_jnt, swing1LimitDegrees=12, swing2LimitDegrees=6, twistLimitDegrees=2}
 {bone1Name=dyn_Hair_r_bk_03_jnt, bone2Name=dyn_Hair_r_bk_02_jnt, swing1LimitDegrees=12, swing2LimitDegrees=6, twistLimitDegrees=2}
 {bone1Name=dyn_Hair_r_bk_04_jnt, bone2Name=dyn_Hair_r_bk_03_jnt, swing1LimitDegrees=12, swing2LimitDegrees=6, twistLimitDegrees=2}
 {bone1Name=dyn_Hair_r_bk_05_jnt, bone2Name=dyn_Hair_r_bk_04_jnt, swing1LimitDegrees=12, swing2LimitDegrees=6, twistLimitDegrees=2}
 {
  bone1Name=dyn_Hair_tp_MSTR
  bone2Name=head
  swing1LimitDegrees=45
  swing1Motion=Free
  swing2LimitDegrees=45
  swing2Motion=Free
  twistLimitDegrees=45
  twistMotion=Free
 }
 {bone1Name=dyn_Hair_tp_01_jnt, bone2Name=dyn_Hair_tp_MSTR, swing1LimitDegrees=12, swing2LimitDegrees=12, twistLimitDegrees=2}
 {bone1Name=dyn_Hair_tp_02_jnt, bone2Name=dyn_Hair_tp_01_jnt, swing1LimitDegrees=12, swing2LimitDegrees=12, twistLimitDegrees=2}
 {bone1Name=dyn_Hair_tp_03_jnt, bone2Name=dyn_Hair_tp_02_jnt, swing1LimitDegrees=12, swing2LimitDegrees=12, twistLimitDegrees=2}
 {bone1Name=dyn_Hair_tp_04_jnt, bone2Name=dyn_Hair_tp_03_jnt, swing1LimitDegrees=12, swing2LimitDegrees=12, twistLimitDegrees=2}
 {bone1Name=dyn_Hair_tp_05_jnt, bone2Name=dyn_Hair_tp_04_jnt, swing1LimitDegrees=12, swing2LimitDegrees=12, twistLimitDegrees=2}
 {bone1Name=dyn_Hair_tp_06_jnt, bone2Name=dyn_Hair_tp_05_jnt, swing1LimitDegrees=12, swing2LimitDegrees=12, twistLimitDegrees=2}
 {
  bone1Name=head
  bone2Name=neck_02
  swing1LimitDegrees=45
  swing1Motion=Free
  swing2LimitDegrees=45
  swing2Motion=Free
  twistLimitDegrees=45
  twistMotion=Free
 }
 {
  bone1Name=neck_02
  bone2Name=neck_01
  swing1LimitDegrees=45
  swing1Motion=Free
  swing2LimitDegrees=45
  swing2Motion=Free
  twistLimitDegrees=45
  twistMotion=Free
 }
 {
  bone1Name=neck_01
  bone2Name=spine_05
  swing1LimitDegrees=45
  swing1Motion=Free
  swing2LimitDegrees=45
  swing2Motion=Free
  twistLimitDegrees=45
  twistMotion=Free
 }
 {
  bone1Name=clavicle_l
  bone2Name=spine_05
  swing1LimitDegrees=45
  swing1Motion=Free
  swing2LimitDegrees=45
  swing2Motion=Free
  twistLimitDegrees=45
  twistMotion=Free
 }
 {
  bone1Name=clavicle_r
  bone2Name=spine_05
  swing1LimitDegrees=45
  swing1Motion=Free
  swing2LimitDegrees=45
  swing2Motion=Free
  twistLimitDegrees=45
  twistMotion=Free
 }
]
